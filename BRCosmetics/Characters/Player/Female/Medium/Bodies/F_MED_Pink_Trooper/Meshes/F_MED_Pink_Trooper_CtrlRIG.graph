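# Control Rig for the F_MED_Pink_Trooper body that procedurally adjusts the holster and its front strap.
# It averages the holsterPivot_zero and holsterStrap_back positions in holsterStrap_front_zero space, using the clamped Z to slide holsterStrap_front.
# An AimBone then points holsterPivot at holsterPivot_tgt, taking its secondary axis from holsterAttachPivot_up, a child of the moved strap.
# The hierarchy also carries suspender, leg flap and wrist strap bones, which this graph leaves untouched.
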ControlRigBlueprint /BRCosmetics/Characters/Player/Female/Medium/Bodies/F_MED_Pink_Trooper/Meshes/F_MED_Pink_Trooper_CtrlRIG
graphs [
 RigVMFunctionLibrary {links [], nodes []}
 RigVMModel {nodes [
   n1 = &RigUnit_BeginExecution()
   n2 = /Script/RigVM.RigVMFunction_MathVectorAdd(
    A=&RigUnit_GetRelativeBoneTransform(Bone=holsterPivot_zero, Space=holsterStrap_front_zero).Transform.Translation
    B=&RigUnit_GetRelativeBoneTransform(Bone=holsterStrap_back, Space=holsterStrap_front_zero).Transform.Translation
   ) {is=struct RigVMUnitNode, template="Add::Execute(in A,in B,out Result)"}
   /Script/ControlRig.RigUnit_SetBoneTranslation(
    Bone=holsterStrap_front
    Space=LocalSpace
    Translation=(X=0,Y=0,Z=0)
    Weight=1
    bPropagateToChildren=True
    ExecutePin=@n1.ExecutePin
    Translation.Z=/Script/RigVM.RigVMFunction_MathFloatClamp(
     Maximum=0
     Minimum=-3
     Value=/Script/RigVM.RigVMFunction_MathVectorDiv(B=(X=2,Y=2,Z=2), A=@n2.Result) {is=struct RigVMUnitNode, template="Divide::Execute(in A,in B,out Result)"}.Result.Z
    ) {is=struct RigVMUnitNode, template="Clamp::Execute(in Value,in Minimum,in Maximum,out Result)"}.Result
   ) {is=struct RigVMUnitNode}
   > /Script/ControlRig.RigUnit_AimBone(
    Bone=holsterPivot
    DebugSettings=(bEnabled=False,Scale=10,WorldOffset=(Rotation=(X=0,Y=0,Z=0,W=1),Translation=(X=0,Y=0,Z=0),Scale3D=(X=1,Y=1,Z=1)))
    Primary=(Weight=1,Axis=(X=1,Y=0,Z=0),Target=(X=0,Y=0,Z=0),Kind=Location,Space="holsterPivot_tgt")
    Secondary=(Weight=1,Axis=(X=0,Y=-1,Z=0),Target=(X=1,Y=0,Z=0),Kind=Location,Space="holsterAttachPivot_up")
    Weight=1
    bPropagateToChildren=True
   ) {is=struct RigVMUnitNode}
  ]}
]
hierarchy [
 root {loc=<0 0 0>, rot=<0 0 0 1>}
 pelvis {parent=root, loc=<0 -0.5301 98.0571>, rot=<0.0127 0.707 -0.0127 -0.707>}
 spine_01 {parent=pelvis, loc=<6.1208 -0.9328 0>, rot=<0 0 0.034 -0.9994>}
 spine_02 {parent=spine_01, loc=<6.5996 -0.2095 0>, rot=<0 0 0.0238 -0.9997>}
 spine_03 {parent=spine_02, loc=<6.4914 -0.1237 0>, rot=<0 0 0.0149 -0.9999>}
 spine_04 {parent=spine_03, loc=<6.9598 0.0794 0>, rot=<0 0 -0.0132 -0.9999>}
 spine_05 {parent=spine_04, loc=<6.9987 0 0>, rot=<0 0 -0.0015 -1>}
 clavicle_l {parent=spine_05, loc=<5.4831 -0.1053 -3.2946>, rot=<0.0465 -0.7289 -0.009 -0.6829>}
 upperarm_l {parent=clavicle_l, loc=<9.8658 -1.2136 -0.4105>, rot=<0.0164 -0.3245 -0.0269 -0.9454>}
 lowerarm_l {parent=upperarm_l, loc=<26.465 -0.8345 0.259>, rot=<0.0036 0.0161 -0.2298 -0.9731>}
 hand_l {parent=lowerarm_l, loc=<22.7355 -0.3203 -0.6105>, rot=<-0.6784 -0.0223 0.0735 -0.7307>}
 index_metacarpal_l {parent=hand_l, loc=<1.6535 0.0183 -2.96>, rot=<0.0158 -0.0124 -0.0173 -0.9996>}
 index_01_l {parent=index_metacarpal_l, loc=<8.3008 0.0357 -0.0524>, rot=<-0.0044 0.0485 0.0924 -0.9945>}
 index_02_l {parent=index_01_l, loc=<3.3587 -0.3003 0.0452>, rot=<-0.002 0.0027 0.3225 -0.9466>}
 index_03_l {parent=index_02_l, loc=<2.5578 -0.123 -0.0252>, rot=<-0.002 0.0067 0.2558 -0.9667>}
 middle_metacarpal_l {parent=hand_l, loc=<2.1856 -0.1725 -0.5984>, rot=<0.006 0.0168 -0.0329 -0.9993>}
 middle_01_l {parent=middle_metacarpal_l, loc=<7.3275 -0.2447 0.0129>, rot=<-0.0208 0.0042 0.0858 -0.9961>}
 middle_02_l {parent=middle_01_l, loc=<4.0123 0.1401 0.1175>, rot=<-0.0218 0.0258 0.3512 -0.9357>}
 middle_03_l {parent=middle_02_l, loc=<2.6077 0.0534 0.1373>, rot=<0.0082 0.0188 0.1708 -0.9851>}
 pinky_metacarpal_l {parent=hand_l, loc=<1.7616 -0.9753 3.3814>, rot=<-0.0397 0.0814 -0.069 -0.9935>}
 pinky_01_l {parent=pinky_metacarpal_l, loc=<7.1709 -0.4831 -0.1094>, rot=<-0.0654 -0.0035 0.2242 -0.9723>}
 pinky_02_l {parent=pinky_01_l, loc=<2.8274 0.0893 -0.1043>, rot=<0.0037 0.0011 0.2854 -0.9584>}
 pinky_03_l {parent=pinky_02_l, loc=<2.2046 0.0308 -0.0078>, rot=<-0.0002 0.0098 0.1445 -0.9895>}
 ring_metacarpal_l {parent=hand_l, loc=<2.2074 -0.3625 1.3638>, rot=<-0.0284 0.0722 -0.0458 -0.9959>}
 ring_01_l {parent=ring_metacarpal_l, loc=<7.0295 -0.6197 -0.1446>, rot=<-0.0489 -0.0252 0.1892 -0.9804>}
 ring_02_l {parent=ring_01_l, loc=<3.6174 0.4647 -0.1027>, rot=<0 0 0.2806 -0.9598>}
 ring_03_l {parent=ring_02_l, loc=<2.4152 0.1141 0.0199>, rot=<0 0 0.1189 -0.9929>}
 thumb_01_l {parent=hand_l, loc=<1.3118 -0.3191 -2.459>, rot=<0.6546 -0.4042 0.0664 -0.6354>}
 thumb_02_l {parent=thumb_01_l, loc=<5.9336 -0.0448 0.0401>, rot=<-0.0037 0.0213 0.298 -0.9543>}
 thumb_03_l {parent=thumb_02_l, loc=<2.2223 0.3682 -0.0907>, rot=<0.0118 -0.013 0.0066 -0.9998>}
 lowerarm_twist_01_l {parent=lowerarm_l, loc=<18.316 0 0>, rot=<0 0 0 1>}
 wristStrap_l_zero {parent=lowerarm_twist_01_l, loc=<3.0896 3.8579 -1.6519>, rot=<0.7343 -0.178 0.6419 -0.1305>}
 wristStrap_l {parent=wristStrap_l_zero, loc=<0 0 0>, rot=<0 0 0 1>}
 lowerarm_twist_02_l {parent=lowerarm_l, loc=<10.6707 0 0>, rot=<0 0 0 1>}
 upperarm_twist_01_l {parent=upperarm_l, loc=<7.2135 -0.1134 0>, rot=<0 0 0 1>}
 upperarm_twist_02_l {parent=upperarm_l, loc=<19.7998 -0.1134 0>, rot=<0 0 0 1>}
 clavicle_r {parent=spine_05, loc=<5.4831 -0.1053 3.2946>, rot=<0.7289 0.0465 -0.6829 0.009>}
 upperarm_r {parent=clavicle_r, loc=<-9.8658 1.2136 0.4104>, rot=<0.0164 -0.3245 -0.0269 -0.9454>}
 lowerarm_r {parent=upperarm_r, loc=<-26.4651 0.8345 -0.259>, rot=<0.0036 0.0161 -0.2298 -0.9731>}
 hand_r {parent=lowerarm_r, loc=<-22.7356 0.3203 0.6106>, rot=<-0.6784 -0.0223 0.0735 -0.7307>}
 index_metacarpal_r {parent=hand_r, loc=<-1.6534 -0.0181 2.9601>, rot=<0.0158 -0.0124 -0.0173 -0.9996>}
 index_01_r {parent=index_metacarpal_r, loc=<-8.3008 -0.0357 0.0524>, rot=<-0.0044 0.0485 0.0924 -0.9945>}
 index_02_r {parent=index_01_r, loc=<-3.3589 0.3003 -0.0453>, rot=<-0.002 0.0027 0.3225 -0.9466>}
 index_03_r {parent=index_02_r, loc=<-2.5578 0.123 0.0252>, rot=<-0.002 0.0067 0.2558 -0.9667>}
 middle_metacarpal_r {parent=hand_r, loc=<-2.1856 0.1726 0.5984>, rot=<0.006 0.0168 -0.0329 -0.9993>}
 middle_01_r {parent=middle_metacarpal_r, loc=<-7.3273 0.2447 -0.0129>, rot=<-0.0208 0.0042 0.0858 -0.9961>}
 middle_02_r {parent=middle_01_r, loc=<-4.0124 -0.14 -0.1175>, rot=<-0.0218 0.0258 0.3512 -0.9357>}
 middle_03_r {parent=middle_02_r, loc=<-2.6077 -0.0533 -0.1373>, rot=<0.0082 0.0188 0.1708 -0.9851>}
 pinky_metacarpal_r {parent=hand_r, loc=<-1.7617 0.9753 -3.3814>, rot=<-0.0397 0.0814 -0.069 -0.9935>}
 pinky_01_r {parent=pinky_metacarpal_r, loc=<-7.1707 0.4832 0.1094>, rot=<-0.0654 -0.0035 0.2242 -0.9723>}
 pinky_02_r {parent=pinky_01_r, loc=<-2.8274 -0.0892 0.1043>, rot=<0.0037 0.0011 0.2854 -0.9584>}
 pinky_03_r {parent=pinky_02_r, loc=<-2.2045 -0.031 0.0078>, rot=<-0.0002 0.0098 0.1445 -0.9895>}
 ring_metacarpal_r {parent=hand_r, loc=<-2.2073 0.3627 -1.3638>, rot=<-0.0284 0.0722 -0.0458 -0.9959>}
 ring_01_r {parent=ring_metacarpal_r, loc=<-7.0297 0.6197 0.1446>, rot=<-0.0489 -0.0252 0.1892 -0.9804>}
 ring_02_r {parent=ring_01_r, loc=<-3.6172 -0.4647 0.1027>, rot=<0 0 0.2806 -0.9598>}
 ring_03_r {parent=ring_02_r, loc=<-2.4152 -0.1141 -0.0199>, rot=<0 0 0.1189 -0.9929>}
 thumb_01_r {parent=hand_r, loc=<-1.3119 0.3192 2.459>, rot=<0.6546 -0.4042 0.0664 -0.6354>}
 thumb_02_r {parent=thumb_01_r, loc=<-5.9336 0.0448 -0.0401>, rot=<-0.0037 0.0213 0.298 -0.9543>}
 thumb_03_r {parent=thumb_02_r, loc=<-2.2223 -0.3682 0.0907>, rot=<0.0118 -0.013 0.0066 -0.9998>}
 lowerarm_twist_01_r {parent=lowerarm_r, loc=<-18.3161 0 0>, rot=<0 0 0 1>}
 wristStrap_r_zero {parent=lowerarm_twist_01_r, loc=<-3.0895 -3.8579 1.652>, rot=<-0.178 -0.7343 -0.1305 -0.6419>}
 wristStrap_r {parent=wristStrap_r_zero, loc=<0 0 0>, rot=<0 0 0 1>}
 lowerarm_twist_02_r {parent=lowerarm_r, loc=<-10.6707 0 0>, rot=<0 0 0 1>}
 upperarm_twist_01_r {parent=upperarm_r, loc=<-7.2135 0.1134 0>, rot=<0 0 0 1>}
 upperarm_twist_02_r {parent=upperarm_r, loc=<-19.7998 0.1134 0>, rot=<0 0 0 1>}
 neck_01 {parent=spine_05, loc=<11.3138 0.1529 0>, rot=<0 0 -0.2003 -0.9797>}
 neck_02 {parent=neck_01, loc=<4.5627 0 0>, rot=<0 0 0.0504 -0.9987>}
 head {parent=neck_02, loc=<5.3914 0 0>, rot=<0 0 0.115 -0.9934>}
 thigh_l {parent=pelvis, loc=<-1.0855 -0.2109 -8.7053>, rot=<0.0009 0.0355 0.0155 -0.9993>}
 calf_l {parent=thigh_l, loc=<-42.5284 -0.0019 -0.7297>, rot=<0.0006 -0.0161 -0.0391 -0.9991>}
 calf_twist_01_l {parent=calf_l, loc=<-35.8483 -1.9197 -0.5794>, rot=<0.0237 0.0095 -0.0035 -0.9997>}
 calf_twist_02_l {parent=calf_l, loc=<-19.1546 -1.7943 -0.2638>, rot=<0.0237 0.0095 -0.0035 -0.9997>}
 legFlap_2_l_zero {parent=calf_twist_02_l, loc=<-2.1679 -3.3965 -5.7327>, rot=<-0.6018 0.3004 0.6804 -0.291>}
 legFlap_2_l {parent=legFlap_2_l_zero, loc=<0 0 0>, rot=<0 0 0 -1>}
 legFlap_1_l_zero {parent=calf_twist_02_l, loc=<-10.8044 -3.5257 -4.346>, rot=<-0.6142 0.3396 0.6316 -0.3295>}
 legFlap_1_l {parent=legFlap_1_l_zero, loc=<0 0 0>, rot=<0 0 0 -1>}
 foot_l {parent=calf_l, loc=<-46.7401 -2.5014 -0.592>, rot=<0.0427 -0.0238 0.0353 -0.9982>}
 ball_l {parent=foot_l, loc=<-3.5658 10.8066 -0.1544>, rot=<0.0001 -0.0001 -0.7063 -0.7079>}
 legFlap_3_l_zero {parent=calf_l, loc=<-12.9117 -5.7186 -6.3347>, rot=<0.6041 -0.3662 -0.5913 0.389>}
 legFlap_3_l {parent=legFlap_3_l_zero, loc=<0 0 0>, rot=<0 0 0 -1>}
 thigh_twist_01_l {parent=thigh_l, loc=<-27.952 0 0>, rot=<0.0013 0 0 -1>}
 suspender_front_0_tgt_zero {parent=thigh_twist_01_l, loc=<-1.2441 9.2816 -9.53>, rot=<-0.0331 0.3277 0.9392 -0.0968>}
 suspender_front_0_tgt {parent=suspender_front_0_tgt_zero, loc=<0 0 0>, rot=<0 0 0 1>}
 suspender_front_1_tgt_zero {parent=thigh_twist_01_l, loc=<-4.9626 13.4759 -11.9996>, rot=<0 0 0 1>}
 suspender_front_1_tgt {parent=suspender_front_1_tgt_zero, loc=<0 0 0>, rot=<0 0 0 1>}
 thigh_shift_l_zero {parent=thigh_twist_01_l, loc=<27.8404 0.0073 0.1923>, rot=<0 0 0 1>}
 thigh_shift_l {parent=thigh_shift_l_zero, loc=<0 0 0>, rot=<0 0 0 1>}
 suspender_up {parent=thigh_l, loc=<-14.5423 -1.3407 -7.9141>, rot=<-0.0019 -0.7317 0.0015 -0.6816>}
 thigh_r {parent=pelvis, loc=<-1.0853 -0.2109 8.7053>, rot=<0.0355 -0.0009 0.9993 0.0155>}
 calf_r {parent=thigh_r, loc=<42.5286 0.0019 0.7297>, rot=<0.0006 -0.0161 -0.0391 -0.9991>}
 calf_twist_01_r {parent=calf_r, loc=<35.8485 1.9197 0.5794>, rot=<0.0237 0.0095 -0.0035 -0.9997>}
 calf_twist_02_r {parent=calf_r, loc=<19.1546 1.7943 0.2638>, rot=<0.0237 0.0095 -0.0035 -0.9997>}
 legFlap_1_r_zero {parent=calf_twist_02_r, loc=<10.8043 3.5257 4.346>, rot=<0.3396 0.6142 -0.3295 -0.6316>}
 legFlap_1_r {parent=legFlap_1_r_zero, loc=<0 0 0>, rot=<0 0 0 1>}
 legFlap_2_r_zero {parent=calf_twist_02_r, loc=<2.1679 3.3965 5.7326>, rot=<0.3004 0.6018 -0.291 -0.6804>}
 legFlap_2_r {parent=legFlap_2_r_zero, loc=<0 0 0>, rot=<0 0 0 1>}
 foot_r {parent=calf_r, loc=<46.7403 2.5014 0.592>, rot=<0.0427 -0.0238 0.0353 -0.9982>}
 ball_r {parent=foot_r, loc=<3.5658 -10.8066 0.1544>, rot=<0.0001 -0.0001 -0.7063 -0.7079>}
 legFlap_3_r_zero {parent=calf_r, loc=<12.9117 5.7186 6.3347>, rot=<0.3662 0.6041 -0.389 -0.5913>}
 legFlap_3_r {parent=legFlap_3_r_zero, loc=<0 0 0>, rot=<0 0 0 1>}
 thigh_twist_01_r {parent=thigh_r, loc=<27.952 0 0>, rot=<0.0013 0 0 -1>}
 thigh_shift_r_zero {parent=thigh_twist_01_r, loc=<-27.952 0 0>, rot=<-0.0013 0 0 -1>}
 thigh_shift_r {parent=thigh_shift_r_zero, loc=<0 0 0>, rot=<0 0 0 -1>}
 hipTriangle_l_zero {parent=pelvis, loc=<0.232 -2.4094 -15.662>, rot=<0.7034 -0.1264 0.6969 0.0597>}
 hipTriangle_l {parent=hipTriangle_l_zero, loc=<0 0 0>, rot=<0 0 0 1>}
 suspender_tgt {parent=pelvis, loc=<10.8206 -8.8221 -0.9078>, rot=<-0.0127 -0.707 0.0127 -0.707>}
 susp_plane_aim_zero {parent=pelvis, loc=<9.4276 -1.8761 -3.9803>, rot=<0.0914 0.608 0.6715 0.4135>}
 susp_plane_aim {parent=susp_plane_aim_zero, loc=<0 0 0>, rot=<0 0 0 1>}
 suspenderThigh_l_zero {parent=susp_plane_aim, loc=<0 0 0>, rot=<0 -0.3827 0 -0.9239>}
 suspenderThigh_l {parent=suspenderThigh_l_zero, loc=<0 0 0>, rot=<0 0 0 1>}
 susp_fb_front {parent=suspenderThigh_l, loc=<2.2369 -1.5574 7.2248>, rot=<0.7071 0 0 -0.7071>}
 suspender_front_0_l_zero {parent=susp_fb_front, loc=<3.3122 -1.0245 0.1712>, rot=<-0.2893 0.0874 0.0527 -0.9518>}
 suspender_front_0_l {parent=suspender_front_0_l_zero, loc=<0 0 0>, rot=<0 0 0 -1>}
 suspender_front_1_l_zero {parent=suspender_front_0_l, loc=<1.1642 -1.5964 -0.0428>, rot=<0.0035 0.0156 0.0291 -0.9994>}
 suspender_front_1_l {parent=suspender_front_1_l_zero, loc=<0 0 0>, rot=<0 0 0 -1>}
 susp_front_mid {parent=suspender_front_1_l, loc=<14.2429 0 0>, rot=<-0.4803 -0.1264 0.0042 -0.8679>}
 suspender_front_2_l {parent=susp_front_mid, loc=<0 0 0>, rot=<0.4657 -0.3488 -0.1463 -0.8>}
 suspender_front_aim_3_l_zero {parent=susp_front_mid, loc=<0 0 0>, rot=<0.0092 -0.068 -0.0002 -0.9976>}
 suspender_front_aim_3_l {parent=suspender_front_aim_3_l_zero, loc=<0 0 0>, rot=<0 0 0 1>}
 suspender_front_4_l {parent=suspender_front_aim_3_l, loc=<0 0 -11.9924>, rot=<0 0 0 1>}
 suspender_front_5_l {parent=suspender_front_4_l, loc=<11.4227 -1.4595 3.6307>, rot=<-0.1518 0.7286 -0.2401 -0.6233>}
 suspender_front_3_l {parent=suspender_front_aim_3_l_zero, loc=<0 0 0>, rot=<0 0 0 1>}
 susp_fb_back {parent=suspenderThigh_l, loc=<-2.2369 1.5574 -7.2248>, rot=<0.7071 0 0 -0.7071>}
 suspender_back_0_l_zero {parent=susp_fb_back, loc=<0 0 0>, rot=<0.9885 0.0993 -0.0324 0.1096>}
 suspender_back_0_l {parent=suspender_back_0_l_zero, loc=<0 0 0>, rot=<0 0 0 1>}
 suspenderBuckle_l_zero {parent=suspender_back_0_l, loc=<9.484 -1.8648 -0.0334>, rot=<0 0.0016 -0.0899 -0.996>}
 suspenderBuckle_l {parent=suspenderBuckle_l_zero, loc=<0 0 0>, rot=<0 0 0 1>}
 susp_back_mid {parent=suspenderBuckle_l, loc=<10.419 0 0>, rot=<0.6231 -0.01 -0.024 -0.7817>}
 suspender_back_1_l_zero {parent=susp_back_mid, loc=<0 0 0>, rot=<-0.5583 0.3337 -0.2024 -0.7321>}
 suspender_back_1_l {parent=suspender_back_1_l_zero, loc=<0 0 0>, rot=<0 0 0 1>}
 suspender_back_0_l_world {parent=susp_plane_aim_zero, loc=<-6.6904 1.5574 -3.527>, rot=<-0.4231 0.1472 0.5369 -0.7149>}
 suspender_front_0_l_world {parent=susp_plane_aim_zero, loc=<6.6904 -1.5574 3.527>, rot=<-0.4231 0.1472 0.5369 -0.7149>}
 pelvis_r {parent=pelvis, loc=<0 0 0>, rot=<0 0 0 -1>}
 pelvis_l {parent=pelvis, loc=<0 0 0>, rot=<0 0 0 -1>}
 holster_zero {parent=pelvis, loc=<4.9099 4.8615 10.9491>, rot=<0.2881 -0.4748 0.8301 -0.05>}
 holster {parent=holster_zero, loc=<15.0724 0 0>, rot=<0 0 0 1>}
 holsterStrap_back {parent=holster, loc=<-2.9144 2.7688 9.5187>, rot=<-0.4556 -0.3257 0.7661 0.3153>}
 holsterStrap_front_zero {parent=holster, loc=<-8.1764 -3.6202 4.7074>, rot=<-0.5106 -0.2461 0.791 0.2303>}
 holsterStrap_front {parent=holsterStrap_front_zero, loc=<0 0 0>, rot=<0 0 0 1>}
 holsterAttachPivot_up {parent=holsterStrap_front, loc=<0 0 0>, rot=<0 0 0 1>}
 holsterPivot_zero {parent=pelvis, loc=<4.91 4.8615 10.9491>, rot=<0.2754 0.6146 -0.2323 -0.7017>}
 holsterPivot {parent=holsterPivot_zero, loc=<0 0 0>, rot=<0.165 0 0 -0.9863>}
 holsterPivot_tgt {parent=holsterPivot_zero, loc=<5 0 0>, rot=<0 0 0 1>}
]
