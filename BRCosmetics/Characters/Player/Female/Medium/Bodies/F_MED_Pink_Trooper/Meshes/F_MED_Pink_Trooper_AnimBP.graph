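# Animation Blueprint for the Pink Trooper cosmetic body, adding secondary motion to suspenders, straps, leg flaps and holster.
# It copies the attached parent mesh's pose, then chains bone offsets, constraints, Look At and spring-based Anim Dynamics with spherical limits into a final Control Rig.
# Each update maps thigh yaw relative to pelvis into alphas driving corrective thigh, suspender and holster adjustments.
# A front-end flag gates extra holster offsets.
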AnimBlueprint /BRCosmetics/Characters/Player/Female/Medium/Bodies/F_MED_Pink_Trooper/Meshes/F_MED_Pink_Trooper_AnimBP
target_skeleton=&/F_MED_Pink_Trooper_Skeleton
graphs [
 AnimGraph {nodes [
   n1 = "Transform (Modify) Bone - Bone: holsterStrap_front_zero"(
    ComponentPose=Local To Component(LocalPose=Copy Pose From Mesh() {bUseAttachedParent=true})
    Alpha=1
    Translation="-0.5,0,0"
    Rotation="0,0,0"
    Scale="1,1,1"
   ) {translationMode=BMM_Additive, alphaBoolBlend {blendOption=Linear}, boneToModify {boneName=holsterStrap_front_zero}, translation {x=-0.5}}
   n2 = "Transform (Modify) Bone - Bone: thigh_shift_l"(ComponentPose=@n1, Alpha=Get thigh_rot_l(), Translation="5,-8,8", Rotation="0,0,0", Scale="1,1,2") {
    scaleMode=BMM_Additive
    scaleSpace=BCS_BoneSpace
    translationMode=BMM_Additive
    translationSpace=BCS_BoneSpace
    alphaBoolBlend {blendOption=Linear}
    alphaScaleBiasClamp {scale=0.6}
    boneToModify {boneName=thigh_shift_l}
    scale {z=2}
    translation=<5 -8 8>
   }
   n3 = "Transform (Modify) Bone - Bone: thigh_shift_r"(ComponentPose=@n2, Alpha=Get thigh_rot_r(), Translation="-5,8,-8", Rotation="0,0,0", Scale="1,1,2") {
    scaleMode=BMM_Additive
    scaleSpace=BCS_BoneSpace
    translationMode=BMM_Additive
    translationSpace=BCS_BoneSpace
    alphaBoolBlend {blendOption=Linear}
    alphaScaleBiasClamp {scale=0.6}
    boneToModify {boneName=thigh_shift_r}
    scale {z=2}
    translation=<-5 8 -8>
   }
   n4 = "Constraint - suspender_front_0_l_world (PPP)"!(ComponentPose=@n3, Alpha=1, ConstraintWeights_0=0.0708, ConstraintWeights_1:float=0.1208, ConstraintWeights_2:float=0.8084) -> (Pose) {
    alphaBoolBlend {blendOption=Linear}
    boneToModify {boneName=suspender_front_0_l_world}
    constraintSetup [{targetBone {boneName=spine_01}}, {targetBone {boneName=spine_02}}, {targetBone {boneName=pelvis}}]
    constraintWeights [1,1,1]
   }
   n5 = "Constraint - suspender_back_0_l_world (PPPP)"!(
    ComponentPose=@n4
    Alpha=1
    ConstraintWeights_0=0.211
    ConstraintWeights_1:float=0.1914
    ConstraintWeights_2:float=0.0184
    ConstraintWeights_3:float=0.5793
   ) -> (Pose) {
    alphaBoolBlend {blendOption=Linear}
    boneToModify {boneName=suspender_back_0_l_world}
    constraintSetup [{targetBone {boneName=spine_01}}, {targetBone {boneName=spine_02}}, {targetBone {boneName=spine_03}}, {targetBone {boneName=pelvis}}]
    constraintWeights [1,1,1,1]
   }
   n6 = Look At - Bone: suspender_front_0_l(ComponentPose=@n5, Alpha=1) {
    alphaBoolBlend {blendOption=Linear}
    boneToModify {boneName=suspender_front_0_l}
    lookAtLocation {x=0}
    lookAtTarget {boneReference {boneName=suspender_front_0_tgt}}
    lookAt_Axis {axis=<1 0>}
   }
   n7 = Look At - Bone: suspender_front_1_l(ComponentPose=@n6, Alpha=Get inv_thigh_rot_l() -> (inv_thigh_rot_l:double)) {
    alphaBoolBlend {blendOption=Linear}
    alphaScaleBiasClamp {bInterpResult=true, interpSpeedDecreasing=5, interpSpeedIncreasing=5}
    boneToModify {boneName=suspender_front_1_l}
    lookAtLocation {x=0}
    lookAtTarget {boneReference {boneName=suspender_front_1_tgt}}
    lookAt_Axis {axis=<1 0>}
   }
   n8 = "Transform (Modify) Bone - Bone: suspender_front_1_l"(ComponentPose=@n7, Alpha=Get thigh_rot_l(), Translation="0,-7,5", Rotation="-40,60,0", Scale="1,1,1") {
    rotationMode=BMM_Additive
    rotationSpace=BCS_BoneSpace
    translationMode=BMM_Additive
    translationSpace=BCS_BoneSpace
    alphaBoolBlend {blendOption=Linear}
    alphaScaleBiasClamp {bInterpResult=true, interpSpeedDecreasing=5, interpSpeedIncreasing=5}
    boneToModify {boneName=suspender_front_1_l}
    rotation {pitch=-40, yaw=60}
    translation {y=-7, z=5}
   }
   n9 = Anim Dynamics - Bone: suspender_front_1_l(ComponentPose=@n8, Alpha=1, ExternalForce="0,0,0") {
    angularDampingOverride=1
    angularSpringConstant=1500
    bAngularSpring=true
    bGravityOverrideInSimSpace=true
    bOverrideAngularDamping=true
    bUseSphericalLimits=true
    gravityScale=0.35
    simulationSpace=RootRelative
    alphaBoolBlend {blendOption=Linear}
    alphaScaleBiasClamp {scale=0.4}
    boundBone {boneName=suspender_front_1_l}
    componentLinearAccScale {z=0.45}
    componentLinearVelScale {z=0.45}
    physicsBodyDefinitions [{
      collisionType=CustomSphere
      sphereCollisionRadius=2
      boundBone {boneName=suspender_front_1_l}
      boxExtents=<10 20 20>
      constraintSetup {angularLimitsMax=<5 80 8>, angularLimitsMin=<-5 0 -40>, angularTarget=<0 0 0>}
      localJointOffset=<15 0 0>
     }]
    sphericalLimits [{limitRadius=9.5, drivingBone {boneName=thigh_l}, sphereLocalOffset=<-15 0 0>}]
   }
   n10 = Anim Dynamics - Bone: suspender_back_0_l(ComponentPose=@n9, Alpha=1, ExternalForce="0,0,0") {
    angularDampingOverride=1
    angularSpringConstant=1500
    bAngularSpring=true
    bGravityOverrideInSimSpace=true
    bOverrideAngularDamping=true
    gravityScale=0.35
    simulationSpace=RootRelative
    alphaBoolBlend {blendOption=Linear}
    alphaScaleBiasClamp {scale=0.8}
    boundBone {boneName=suspender_back_0_l}
    componentLinearAccScale {z=0.45}
    componentLinearVelScale {z=0.45}
    physicsBodyDefinitions [{
      sphereCollisionRadius=0
      boundBone {boneName=suspender_back_0_l}
      boxExtents=<10 20 20>
      constraintSetup {angularLimitsMax=<5 5 0>, angularLimitsMin=<-5 -40 -40>, angularTarget=<0 0 0>}
      localJointOffset=<10 0 0>
     }]
   }
   n11 = Anim Dynamics - Bone: suspenderBuckle_l(ComponentPose=@n10, Alpha=1, ExternalForce="0,0,0") {
    angularDampingOverride=1
    angularSpringConstant=1500
    bAngularSpring=true
    bGravityOverrideInSimSpace=true
    bOverrideAngularDamping=true
    bUseSphericalLimits=true
    gravityScale=0.35
    simulationSpace=RootRelative
    alphaBoolBlend {blendOption=Linear}
    alphaScaleBiasClamp {scale=0.8}
    boundBone {boneName=suspenderBuckle_l}
    componentLinearAccScale {z=0.45}
    componentLinearVelScale {z=0.45}
    physicsBodyDefinitions [{
      collisionType=CustomSphere
      sphereCollisionRadius=1.5
      boundBone {boneName=suspenderBuckle_l}
      boxExtents=<10 20 20>
      constraintSetup {angularLimitsMax=<10 0 40>, angularLimitsMin=<-10 -15 -40>, angularTarget=<0 0 0>}
      localJointOffset=<4 0 0>
     }]
    sphericalLimits [{limitRadius=8, drivingBone {boneName=pelvis_l}, sphereLocalOffset=<-2 -5 -4>}]
   }
   n12 = Look At - Bone: suspender_front_aim_3_l(ComponentPose=@n11, Alpha=1) {
    alphaBoolBlend {blendOption=Linear}
    boneToModify {boneName=suspender_front_aim_3_l}
    lookAtLocation {x=0}
    lookAtTarget {boneReference {boneName=susp_back_mid}}
    lookAt_Axis {axis {y=0, z=-1}}
   }
   n13 = "Constraint - suspender_front_4_l (TT)"!(ComponentPose=@n12, Alpha=1, ConstraintWeights_0=0.5, ConstraintWeights_1:float=0.5) -> (Pose) {
    alphaBoolBlend {blendOption=Linear}
    boneToModify {boneName=suspender_front_4_l}
    constraintSetup [
     {offsetOption=None, transformType=Translation, targetBone {boneName=susp_front_mid}}
     {offsetOption=None, transformType=Translation, targetBone {boneName=susp_back_mid}}
    ]
    constraintWeights [1,1]
   }
   n14 = Anim Dynamics - Bone: suspender_front_4_l(ComponentPose=@n13, Alpha=1, ExternalForce="0,0,0") {
    angularDampingOverride=1
    angularSpringConstant=1500
    bAngularSpring=true
    bGravityOverrideInSimSpace=true
    bOverrideAngularDamping=true
    bUseSphericalLimits=true
    gravityScale=0.35
    simulationSpace=RootRelative
    alphaBoolBlend {blendOption=Linear}
    alphaScaleBiasClamp {interpSpeedDecreasing=5, interpSpeedIncreasing=5}
    boundBone {boneName=suspender_front_4_l}
    componentLinearAccScale {z=0.45}
    componentLinearVelScale {z=0.45}
    physicsBodyDefinitions [{
      collisionType=CustomSphere
      sphereCollisionRadius=2
      boundBone {boneName=suspender_front_4_l}
      boxExtents=<10 20 20>
      constraintSetup {angularLimitsMax=<10 20 0>, angularLimitsMin=<-5 -20 -80>, angularTarget=<0 0 0>}
      localJointOffset=<12 0 0>
     }]
    sphericalLimits [{limitRadius=10, drivingBone {boneName=thigh_l}, sphereLocalOffset=<-15 0 0>}]
   }
   n15 = Get thigh_rot_l()
   n16 = Anim Dynamics - Bone: suspender_front_4_l(ComponentPose=@n14, Alpha=@n15, ExternalForce="0,0,0") {
    angularDampingOverride=1
    angularSpringConstant=1500
    bAngularSpring=true
    bGravityOverrideInSimSpace=true
    bOverrideAngularDamping=true
    bUseSphericalLimits=true
    gravityScale=0.35
    simulationSpace=RootRelative
    alphaBoolBlend {blendOption=Linear}
    alphaScaleBiasClamp {interpSpeedDecreasing=5, interpSpeedIncreasing=5}
    boundBone {boneName=suspender_front_4_l}
    componentLinearAccScale {z=0.45}
    componentLinearVelScale {z=0.45}
    physicsBodyDefinitions [{
      collisionType=CustomSphere
      sphereCollisionRadius=2
      boundBone {boneName=suspender_front_4_l}
      boxExtents=<10 20 20>
      constraintSetup {angularLimitsMax=<0 0 80>, angularLimitsMin=<0 0 0>, angularTarget=<0 0 0>}
      localJointOffset=<12 0 0>
     }]
    sphericalLimits [{limitRadius=10, drivingBone {boneName=thigh_l}, sphereLocalOffset=<-15 0 0>}]
   }
   n17 = "Transform (Modify) Bone - Bone: suspender_front_3_l"(ComponentPose=@n16, Alpha=@n15, Translation="0,0,0", Rotation="-50,0,0", Scale="1,1,1") {
    rotationMode=BMM_Additive
    rotationSpace=BCS_BoneSpace
    alphaBoolBlend {blendOption=Linear}
    alphaScaleBiasClamp {bInterpResult=true, interpSpeedDecreasing=5, interpSpeedIncreasing=5}
    boneToModify {boneName=suspender_front_3_l}
    rotation {pitch=-50}
   }
   n18 = Anim Dynamics - Bone: wristStrap_l!(ComponentPose=@n17, Alpha=1, GravityOverride:Vector="0,0,0", ExternalForce="0,0,0") -> (Pose) {
    angularBiasOverride=0.99
    angularDampingOverride=0.99
    angularSpringConstant=1500
    bAngularSpring=true
    bOverrideAngularBias=true
    bOverrideAngularDamping=true
    bUseGravityOverride=true
    gravityScale=0.35
    linearDampingOverride=0.9
    simulationSpace=RootRelative
    alphaBoolBlend {blendOption=Linear}
    boundBone {boneName=wristStrap_l}
    gravityOverride=<1 1 1>
    physicsBodyDefinitions [{
      sphereCollisionRadius=0
      boundBone {boneName=wristStrap_l}
      boxExtents=<10 15 15>
      constraintSetup {angularLimitsMax=<0 10 30>, angularLimitsMin=<0 -10 -30>, angularTarget=<1 0 0>}
      localJointOffset=<1 0 0>
     }]
   }
   n19 = Anim Dynamics - Bone: wristStrap_r!(ComponentPose=@n18, Alpha=1, GravityOverride:Vector="0,0,0", ExternalForce="0,0,0") -> (Pose) {
    angularBiasOverride=0.99
    angularDampingOverride=0.99
    angularSpringConstant=1500
    bAngularSpring=true
    bOverrideAngularBias=true
    bOverrideAngularDamping=true
    bUseGravityOverride=true
    gravityScale=0.35
    linearDampingOverride=0.9
    simulationSpace=RootRelative
    alphaBoolBlend {blendOption=Linear}
    boundBone {boneName=wristStrap_r}
    gravityOverride=<1 1 1>
    physicsBodyDefinitions [{
      sphereCollisionRadius=0
      boundBone {boneName=wristStrap_r}
      boxExtents=<10 15 15>
      constraintSetup {angularLimitsMax=<0 10 30>, angularLimitsMin=<0 -10 -30>, angularTarget=<1 0 0>}
      localJointOffset=<1 0 0>
     }]
   }
   n20 = Anim Dynamics - Bone: legFlap_1_l!(ComponentPose=@n19, Alpha=1, GravityOverride:Vector="0,0,0", ExternalForce="0,0,0") -> (Pose) {
    angularBiasOverride=0.99
    angularDampingOverride=0.99
    angularSpringConstant=1500
    bAngularSpring=true
    bOverrideAngularBias=true
    bOverrideAngularDamping=true
    bUseGravityOverride=true
    gravityScale=0.35
    linearDampingOverride=0.9
    simulationSpace=RootRelative
    alphaBoolBlend {blendOption=Linear}
    boundBone {boneName=legFlap_1_l}
    gravityOverride=<1 1 1>
    physicsBodyDefinitions [{
      sphereCollisionRadius=0
      boundBone {boneName=legFlap_1_l}
      boxExtents=<10 15 15>
      constraintSetup {angularLimitsMax=<0 10 30>, angularLimitsMin=<0 -10 -30>, angularTarget=<1 0 0>}
      localJointOffset=<1 0 0>
     }]
   }
   n21 = Anim Dynamics - Bone: legFlap_2_l!(ComponentPose=@n20, Alpha=1, GravityOverride:Vector="0,0,0", ExternalForce="0,0,0") -> (Pose) {
    angularBiasOverride=0.99
    angularDampingOverride=0.99
    angularSpringConstant=1500
    bAngularSpring=true
    bOverrideAngularBias=true
    bOverrideAngularDamping=true
    bUseGravityOverride=true
    gravityScale=0.35
    linearDampingOverride=0.9
    simulationSpace=RootRelative
    alphaBoolBlend {blendOption=Linear}
    boundBone {boneName=legFlap_2_l}
    gravityOverride=<1 1 1>
    physicsBodyDefinitions [{
      sphereCollisionRadius=0
      boundBone {boneName=legFlap_2_l}
      boxExtents=<12 17 17>
      constraintSetup {angularLimitsMax=<0 10 30>, angularLimitsMin=<0 -10 -30>, angularTarget=<1 0 0>}
      localJointOffset=<1 0 0>
     }]
   }
   n22 = Anim Dynamics - Bone: legFlap_3_l!(ComponentPose=@n21, Alpha=1, GravityOverride:Vector="0,0,0", ExternalForce="0,0,0") -> (Pose) {
    angularBiasOverride=0.99
    angularDampingOverride=0.99
    angularSpringConstant=2000
    bAngularSpring=true
    bOverrideAngularBias=true
    bOverrideAngularDamping=true
    bUseGravityOverride=true
    gravityScale=0.35
    linearDampingOverride=0.9
    simulationSpace=RootRelative
    alphaBoolBlend {blendOption=Linear}
    boundBone {boneName=legFlap_3_l}
    gravityOverride=<1 1 1>
    physicsBodyDefinitions [{
      sphereCollisionRadius=0
      boundBone {boneName=legFlap_3_l}
      boxExtents=<12 17 17>
      constraintSetup {angularLimitsMax=<0 10 50>, angularLimitsMin=<0 -10 -50>, angularTarget=<1 0 0>}
      localJointOffset=<2 0 0>
     }]
   }
   n23 = Anim Dynamics - Bone: legFlap_1_r!(ComponentPose=@n22, Alpha=1, GravityOverride:Vector="0,0,0", ExternalForce="0,0,0") -> (Pose) {
    angularBiasOverride=0.99
    angularDampingOverride=0.99
    angularSpringConstant=1500
    bAngularSpring=true
    bOverrideAngularBias=true
    bOverrideAngularDamping=true
    bUseGravityOverride=true
    gravityScale=0.35
    linearDampingOverride=0.9
    simulationSpace=RootRelative
    alphaBoolBlend {blendOption=Linear}
    boundBone {boneName=legFlap_1_r}
    gravityOverride=<1 1 1>
    physicsBodyDefinitions [{
      sphereCollisionRadius=0
      boundBone {boneName=legFlap_1_r}
      boxExtents=<10 15 15>
      constraintSetup {angularLimitsMax=<0 10 30>, angularLimitsMin=<0 -10 -30>, angularTarget=<1 0 0>}
      localJointOffset=<1 0 0>
     }]
   }
   n24 = Anim Dynamics - Bone: legFlap_2_r!(ComponentPose=@n23, Alpha=1, GravityOverride:Vector="0,0,0", ExternalForce="0,0,0") -> (Pose) {
    angularBiasOverride=0.99
    angularDampingOverride=0.99
    angularSpringConstant=1500
    bAngularSpring=true
    bOverrideAngularBias=true
    bOverrideAngularDamping=true
    bUseGravityOverride=true
    gravityScale=0.35
    linearDampingOverride=0.9
    simulationSpace=RootRelative
    alphaBoolBlend {blendOption=Linear}
    boundBone {boneName=legFlap_2_r}
    gravityOverride=<1 1 1>
    physicsBodyDefinitions [{
      sphereCollisionRadius=0
      boundBone {boneName=legFlap_2_r}
      boxExtents=<12 17 17>
      constraintSetup {angularLimitsMax=<0 10 30>, angularLimitsMin=<0 -10 -30>, angularTarget=<1 0 0>}
      localJointOffset=<1 0 0>
     }]
   }
   n25 = Anim Dynamics - Bone: legFlap_3_r!(ComponentPose=@n24, Alpha=1, GravityOverride:Vector="0,0,0", ExternalForce="0,0,0") -> (Pose) {
    angularBiasOverride=0.99
    angularDampingOverride=0.99
    angularSpringConstant=2000
    bAngularSpring=true
    bOverrideAngularBias=true
    bOverrideAngularDamping=true
    bUseGravityOverride=true
    gravityScale=0.35
    linearDampingOverride=0.9
    simulationSpace=RootRelative
    alphaBoolBlend {blendOption=Linear}
    boundBone {boneName=legFlap_3_r}
    gravityOverride=<1 1 1>
    physicsBodyDefinitions [{
      sphereCollisionRadius=0
      boundBone {boneName=legFlap_3_r}
      boxExtents=<12 17 17>
      constraintSetup {angularLimitsMax=<0 10 50>, angularLimitsMin=<0 -10 -50>, angularTarget=<1 0 0>}
      localJointOffset=<2 0 0>
     }]
   }
   n26 = Anim Dynamics - Bone: wristStrap_l!(ComponentPose=@n25, Alpha=1, ExternalForce="0,0,0") -> (Pose) {
    angularDampingOverride=1
    angularSpringConstant=1500
    bAngularSpring=true
    bGravityOverrideInSimSpace=true
    bOverrideAngularDamping=true
    gravityScale=0.35
    simulationSpace=RootRelative
    alphaBoolBlend {blendOption=Linear}
    alphaScaleBiasClamp {scale=0.7}
    boundBone {boneName=wristStrap_l}
    componentLinearAccScale {z=0.43}
    componentLinearVelScale {z=0.43}
    physicsBodyDefinitions [{
      sphereCollisionRadius=0
      boundBone {boneName=wristStrap_l}
      boxExtents=<10 20 20>
      constraintSetup {angularLimitsMax=<0 30 10>, angularLimitsMin=<0 -30 -50>, angularTarget=<1 0 0>}
      localJointOffset=<10 0 0>
     }]
   }
   n27 = Anim Dynamics - Bone: wristStrap_r!(ComponentPose=@n26, Alpha=1, ExternalForce="0,0,0") -> (Pose) {
    angularDampingOverride=1
    angularSpringConstant=1500
    bAngularSpring=true
    bGravityOverrideInSimSpace=true
    bOverrideAngularDamping=true
    gravityScale=0.35
    simulationSpace=RootRelative
    alphaBoolBlend {blendOption=Linear}
    alphaScaleBiasClamp {scale=0.7}
    boundBone {boneName=wristStrap_r}
    componentLinearAccScale {z=0.43}
    componentLinearVelScale {z=0.43}
    physicsBodyDefinitions [{
      sphereCollisionRadius=0
      boundBone {boneName=wristStrap_r}
      boxExtents=<10 20 20>
      constraintSetup {angularLimitsMax=<0 30 10>, angularLimitsMin=<0 -30 -50>, angularTarget=<1 0 0>}
      localJointOffset=<10 0 0>
     }]
   }
   n28 = "Constraint - hipTriangle_l_zero (PPP)"!(ComponentPose=@n27, Alpha=1, ConstraintWeights_0=0.5356, ConstraintWeights_1:float=0.4136, ConstraintWeights_2:float=0.0508) -> (Pose) {
    alphaBoolBlend {blendOption=Linear}
    boneToModify {boneName=hipTriangle_l_zero}
    constraintSetup [{targetBone {boneName=pelvis}}, {targetBone {boneName=thigh_twist_01_l}}, {targetBone {boneName=thigh_l}}]
    constraintWeights [1,1,1]
   }
   n29 = "Constraint - holster (PPPP)"!(
    ComponentPose=@n28
    Alpha=1
    ConstraintWeights_0=0.1726
    ConstraintWeights_1:float=0.3271
    ConstraintWeights_2:float=0.2559
    ConstraintWeights_3:float=0.2444
   ) -> (Pose) {
    alphaBoolBlend {blendOption=Linear}
    boneToModify {boneName=holster}
    constraintSetup [
     {targetBone {boneName=thigh_r}}
     {targetBone {boneName=pelvis}}
     {targetBone {boneName=thigh_twist_01_r}}
     {targetBone {boneName=thigh_shift_r}}
    ]
    constraintWeights [1,1,1,1]
   }
   n30 = "Constraint - holsterPivot_zero (PPPP)"!(
    ComponentPose=@n29
    Alpha=1
    ConstraintWeights_0=0.0145
    ConstraintWeights_1:float=0.0247
    ConstraintWeights_2:float=0.0266
    ConstraintWeights_3:float=0.9341
   ) -> (Pose) {
    alphaBoolBlend {blendOption=Linear}
    boneToModify {boneName=holsterPivot_zero}
    constraintSetup [{targetBone {boneName=spine_01}}, {targetBone {boneName=spine_02}}, {targetBone {boneName=thigh_twist_01_r}}, {targetBone {boneName=pelvis}}]
    constraintWeights [1,1,1,1]
   }
   n31 = "Transform (Modify) Bone - Bone: holster"(ComponentPose=@n30, Alpha=Get improve_movement() -> (improve_movement:double), Translation="-5,0,0", Rotation="0,0,0", Scale="1,1,1") {
    rotationMode=BMM_Additive
    rotationSpace=BCS_BoneSpace
    translationMode=BMM_Additive
    translationSpace=BCS_BoneSpace
    alphaBoolBlend {blendOption=Linear}
    boneToModify {boneName=holster}
    translation {x=-5}
   }
   n32 = "Transform (Modify) Bone - Bone: holster"!(ComponentPose=@n31, bAlphaBoolEnabled:bool=Get bIsInFrontEnd(), Translation="2,5,5", Rotation="0,10,30", Scale="1,1,1") -> (Pose) {
    alphaInputType=Bool
    rotationMode=BMM_Additive
    rotationSpace=BCS_BoneSpace
    translationMode=BMM_Additive
    translationSpace=BCS_BoneSpace
    alphaBoolBlend {blendOption=Linear}
    boneToModify {boneName=holster}
    rotation {roll=30, yaw=10}
    translation=<2 5 5>
   }
   n33 = Control Rig(Source=Component To Local(ComponentPose=@n32)) {
    alphaBoolBlend {blendOption=Linear}
    controlRigAssetReference {blueprintRigClass=&/F_MED_Pink_Trooper_CtrlRIG.F_MED_Pink_Trooper_CtrlRIG_C}
   }
   Output Pose(Result=@n33) {layerGroup=None}
  ]}
 EventGraph {nodes [
   Event BlueprintUpdateAnimation()
   n1 = Get bIsInFrontEnd()
   n2 = GetDeltaTransformFromRefPose(self=GetOwnerSkeletalMeshComponent(), BoneName=thigh_r, BaseName=pelvis)
   n3 = MapRangeClamped(Value=BreakRotator(InRot=BreakTransform(InTransform=@n2).Rotation).Yaw, InRangeA=0, InRangeB=-90)
   > Set improve_movement(improve_movement:double=float * float(A=@n3, B="To Float (Boolean)"(InBool=@n1)) {is=title K2Node_PromotableOperator}) -> (Output_Get:double)
   > Set thigh_rot_r(thigh_rot_r=@n3)
   n4 = GetDeltaTransformFromRefPose(self=GetOwnerSkeletalMeshComponent(), BoneName=thigh_l, BaseName=pelvis)
   n5 = MapRangeClamped(Value=BreakRotator(InRot=BreakTransform(InTransform=@n4).Rotation).Yaw, InRangeA=-15, InRangeB=-110)
   > Set thigh_rot_l(thigh_rot_l=@n5)
   > Set inv_thigh_rot_l(inv_thigh_rot_l:double=float - float(B=@n5) {is=title K2Node_CallFunction}) -> (Output_Get:double)
  ]}
]
variables [
 improve_movement {type=number}
 inv_thigh_rot_l {type=number}
 thigh_rot_l {type=number}
 thigh_rot_r {type=number}
]
