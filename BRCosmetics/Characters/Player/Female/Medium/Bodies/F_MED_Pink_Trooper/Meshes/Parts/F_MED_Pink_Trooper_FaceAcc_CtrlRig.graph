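# Control Rig for the Pink Trooper female FaceAcc part that keeps selected hair bones out of the upper body.
# Spheres offset from spine_05 and the upper arms act as colliders, using max(distance, radius) to push points outward.
# Each block aims a strand bone at the corrected end position, usually clamps its tip translation, and draws debug axes.
# The hair_3_r, right-arm hair_5_r and hair_1_c_tip corrections lack execution links from Begin Execution.
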ControlRigBlueprint /BRCosmetics/Characters/Player/Female/Medium/Bodies/F_MED_Pink_Trooper/Meshes/Parts/F_MED_Pink_Trooper_FaceAcc_CtrlRig
graphs [
 RigVMFunctionLibrary {links [], nodes []}
 RigVMModel {nodes [
   n1 = &RigVMFunction_MathTransformMul(
    A=(Rotation=(X=0,Y=0,Z=0,W=1),Translation=(X=-2,Y=3,Z=0),Scale3D=(X=1,Y=1,Z=1))
    B=&RigUnit_GetBoneTransform(Bone=spine_05).Transform
   ) {template="Multiply::Execute(in A,in B,out Result)"}
   n2 = /Script/RigVM.RigVMFunction_MathFloatAbs(Value=20) {is=struct RigVMUnitNode, template="Absolute::Execute(in Value,out Result)"}
   n3 = /Script/RigVM.RigVMFunction_MathVectorSub(A=&RigUnit_GetBoneTransform(Bone=hair_3_r_end).Transform.Translation, B=@n1.Result.Translation) {is=struct RigVMUnitNode, template="Subtract::Execute(in A,in B,out Result)"}
   n4 = /Script/RigVM.RigVMFunction_MathVectorFromFloat(Value=/Script/RigVM.RigVMFunction_MathFloatMax(
     B=@n2.Result
     A=/Script/RigVM.RigVMFunction_MathVectorLength(Value=@n3.Result) {is=struct RigVMUnitNode, template="Length::Execute(in Value,out Result)"}.Result
    ) {is=struct RigVMUnitNode, template="Maximum::Execute(in A,in B,out Result)"}.Result) {is=struct RigVMUnitNode, template="Cast::Execute(in Value,out Result)"}
   n5 = /Script/RigVM.RigVMFunction_MathVectorUnit(Value=@n3.Result) {is=struct RigVMUnitNode, template="Unit::Execute(in Value,out Result)"}
   /Script/ControlRig.RigUnit_AimBone(
    Bone=hair_3_r
    DebugSettings=(bEnabled=False,Scale=10,WorldOffset=(Rotation=(X=0,Y=0,Z=0,W=1),Translation=(X=0,Y=0,Z=0),Scale3D=(X=1,Y=1,Z=1)))
    ExecutePin=()
    Primary=(Weight=1,Axis=(X=1,Y=0,Z=0),Target=(X=1,Y=0,Z=0),Kind=Location,Space="None")
    Secondary=(Weight=0,Axis=(X=0,Y=0,Z=1),Target=(X=1,Y=0,Z=0),Kind=Direction,Space="None")
    Weight=1
    bPropagateToChildren=True
    Primary.Target=/Script/RigVM.RigVMFunction_MathVectorAdd(
     B=@n1.Result.Translation
     A=/Script/RigVM.RigVMFunction_MathVectorMul(A=@n4.Result, B=@n5.Result) {is=struct RigVMUnitNode, template="Multiply::Execute(in A,in B,out Result)"}.Result
    ) {is=struct RigVMUnitNode, template="Add::Execute(in A,in B,out Result)"}.Result
   ) {is=struct RigVMUnitNode}
   > /Script/ControlRig.RigUnit_DebugTransformMutable(
    Color=(R=1,G=0,B=0,A=1)
    DebugDrawSettings=(DepthPriority=SDPG_Foreground,Lifetime=-1)
    Mode=Axes
    Scale=10
    Space=None
    Thickness=0
    WorldOffset=(Rotation=(X=0,Y=0,Z=0,W=1),Translation=(X=0,Y=0,Z=0),Scale3D=(X=1,Y=1,Z=1))
    bEnabled=True
    Transform=@n1.Result
   ) {is=struct RigVMUnitNode}
   n6 = /Script/ControlRig.RigUnit_BoneName(Bone=hair_3_r) {is=struct RigVMUnitNode}
   n7 = &RigVMFunction_MathVectorSub(A=&RigUnit_GetBoneTransform(Bone=@n6.Bone).Transform.Translation, B=@n1.Result.Translation) {template="Subtract::Execute(in A,in B,out Result)"}
   n8 = &RigVMFunction_MathVectorFromFloat(Value=&RigVMFunction_MathFloatMax(B=@n2.Result, A=&RigVMFunction_MathVectorLength(Value=@n7.Result) {template="Length::Execute(in Value,out Result)"}.Result) {template="Maximum::Execute(in A,in B,out Result)"}.Result) {template="Cast::Execute(in Value,out Result)"}
   n9 = &RigVMFunction_MathVectorUnit(Value=@n7.Result) {template="Unit::Execute(in Value,out Result)"}
   > /Script/ControlRig.RigUnit_SetBoneTranslation(
    Space=GlobalSpace
    Weight=1
    bPropagateToChildren=True
    Bone=@n6.Bone
    Translation=&RigVMFunction_MathVectorAdd(
     B=@n1.Result.Translation
     A=&RigVMFunction_MathVectorMul(A=@n8.Result, B=@n9.Result) {template="Multiply::Execute(in A,in B,out Result)"}.Result
    ) {template="Add::Execute(in A,in B,out Result)"}.Result
   ) {is=struct RigVMUnitNode}
   n10 = &RigVMFunction_MathTransformMul(
    A=(Rotation=(X=0,Y=0,Z=0,W=1),Translation=(X=-13,Y=-3,Z=29),Scale3D=(X=1,Y=1,Z=1))
    B=&RigUnit_GetBoneTransform(Bone=upperarm_r).Transform
   ) {template="Multiply::Execute(in A,in B,out Result)"}
   n11 = &RigVMFunction_MathFloatAbs(Value=39) {template="Absolute::Execute(in Value,out Result)"}
   n12 = &RigVMFunction_MathVectorSub(A=&RigUnit_GetBoneTransform(Bone=hair_5_r_end).Transform.Translation, B=@n10.Result.Translation) {template="Subtract::Execute(in A,in B,out Result)"}
   n13 = &RigVMFunction_MathVectorFromFloat(Value=&RigVMFunction_MathFloatMax(B=@n11.Result, A=&RigVMFunction_MathVectorLength(Value=@n12.Result) {template="Length::Execute(in Value,out Result)"}.Result) {template="Maximum::Execute(in A,in B,out Result)"}.Result) {template="Cast::Execute(in Value,out Result)"}
   n14 = &RigVMFunction_MathVectorUnit(Value=@n12.Result) {template="Unit::Execute(in Value,out Result)"}
   &RigUnit_AimBone(
    Bone=hair_5_cst_r
    DebugSettings=(bEnabled=False,Scale=10,WorldOffset=(Rotation=(X=0,Y=0,Z=0,W=1),Translation=(X=0,Y=0,Z=0),Scale3D=(X=1,Y=1,Z=1)))
    ExecutePin=()
    Primary=(Weight=1,Axis=(X=1,Y=0,Z=0),Target=(X=1,Y=0,Z=0),Kind=Location,Space="None")
    Secondary=(Weight=0,Axis=(X=0,Y=0,Z=1),Target=(X=1,Y=0,Z=0),Kind=Direction,Space="None")
    Weight=1
    bPropagateToChildren=True
    Primary.Target=&RigVMFunction_MathVectorAdd(
     B=@n10.Result.Translation
     A=&RigVMFunction_MathVectorMul(A=@n13.Result, B=@n14.Result) {template="Multiply::Execute(in A,in B,out Result)"}.Result
    ) {template="Add::Execute(in A,in B,out Result)"}.Result
   )
   > &RigUnit_DebugTransformMutable(
    Color=(R=1,G=0,B=0,A=1)
    DebugDrawSettings=(DepthPriority=SDPG_Foreground,Lifetime=-1)
    Mode=Axes
    Scale=10
    Space=None
    Thickness=0
    WorldOffset=(Rotation=(X=0,Y=0,Z=0,W=1),Translation=(X=0,Y=0,Z=0),Scale3D=(X=1,Y=1,Z=1))
    bEnabled=True
    Transform=@n10.Result
   )
   n15 = &RigUnit_BoneName(Bone=hair_5_r_tip)
   n16 = &RigVMFunction_MathVectorSub(A=&RigUnit_GetBoneTransform(Bone=@n15.Bone).Transform.Translation, B=@n10.Result.Translation) {template="Subtract::Execute(in A,in B,out Result)"}
   n17 = &RigVMFunction_MathVectorFromFloat(Value=&RigVMFunction_MathFloatMax(B=@n11.Result, A=&RigVMFunction_MathVectorLength(Value=@n16.Result) {template="Length::Execute(in Value,out Result)"}.Result) {template="Maximum::Execute(in A,in B,out Result)"}.Result) {template="Cast::Execute(in Value,out Result)"}
   n18 = &RigVMFunction_MathVectorUnit(Value=@n16.Result) {template="Unit::Execute(in Value,out Result)"}
   > &RigUnit_SetBoneTranslation(
    Space=GlobalSpace
    Weight=1
    bPropagateToChildren=True
    Bone=@n15.Bone
    Translation=&RigVMFunction_MathVectorAdd(
     B=@n10.Result.Translation
     A=&RigVMFunction_MathVectorMul(A=@n17.Result, B=@n18.Result) {template="Multiply::Execute(in A,in B,out Result)"}.Result
    ) {template="Add::Execute(in A,in B,out Result)"}.Result
   )
   n19 = &RigUnit_BeginExecution()
   n20 = &RigVMFunction_MathTransformMul(
    A=(Rotation=(X=0,Y=0,Z=0,W=1),Translation=(X=-2,Y=3,Z=0),Scale3D=(X=1,Y=1,Z=1))
    B=&RigUnit_GetBoneTransform(Bone=spine_05).Transform
   ) {template="Multiply::Execute(in A,in B,out Result)"}
   n21 = &RigVMFunction_MathFloatAbs(Value=15) {template="Absolute::Execute(in Value,out Result)"}
   n22 = &RigVMFunction_MathVectorSub(A=&RigUnit_GetBoneTransform(Bone=hair_4_r_end).Transform.Translation, B=@n20.Result.Translation) {template="Subtract::Execute(in A,in B,out Result)"}
   n23 = &RigVMFunction_MathVectorFromFloat(Value=&RigVMFunction_MathFloatMax(B=@n21.Result, A=&RigVMFunction_MathVectorLength(Value=@n22.Result) {template="Length::Execute(in Value,out Result)"}.Result) {template="Maximum::Execute(in A,in B,out Result)"}.Result) {template="Cast::Execute(in Value,out Result)"}
   n24 = &RigVMFunction_MathVectorUnit(Value=@n22.Result) {template="Unit::Execute(in Value,out Result)"}
   &RigUnit_AimBone(
    Bone=hair_4_r
    DebugSettings=(bEnabled=False,Scale=10,WorldOffset=(Rotation=(X=0,Y=0,Z=0,W=1),Translation=(X=0,Y=0,Z=0),Scale3D=(X=1,Y=1,Z=1)))
    Primary=(Weight=1,Axis=(X=1,Y=0,Z=0),Target=(X=1,Y=0,Z=0),Kind=Location,Space="None")
    Secondary=(Weight=0,Axis=(X=0,Y=0,Z=1),Target=(X=1,Y=0,Z=0),Kind=Direction,Space="None")
    Weight=1
    bPropagateToChildren=True
    ExecutePin=@n19.ExecutePin
    Primary.Target=&RigVMFunction_MathVectorAdd(
     B=@n20.Result.Translation
     A=&RigVMFunction_MathVectorMul(A=@n23.Result, B=@n24.Result) {template="Multiply::Execute(in A,in B,out Result)"}.Result
    ) {template="Add::Execute(in A,in B,out Result)"}.Result
   )
   > /Script/ControlRig.RigUnit_DebugArc(
    Color=(R=1,G=0,B=0,A=1)
    DebugDrawSettings=(DepthPriority=SDPG_Foreground,Lifetime=-1)
    Detail=16
    MaximumDegrees=360
    MinimumDegrees=0
    Space=None
    Thickness=0
    WorldOffset=(Rotation=(X=0,Y=0,Z=0,W=1),Translation=(X=0,Y=0,Z=0),Scale3D=(X=1,Y=1,Z=1))
    bEnabled=True
    Radius=@n21.Result
    Transform=@n20.Result
   ) {is=struct RigVMUnitNode}
   > &RigUnit_DebugTransformMutable(
    Color=(R=1,G=0,B=0,A=1)
    DebugDrawSettings=(DepthPriority=SDPG_Foreground,Lifetime=-1)
    Mode=Axes
    Scale=10
    Space=None
    Thickness=0
    WorldOffset=(Rotation=(X=0,Y=0,Z=0,W=1),Translation=(X=0,Y=0,Z=0),Scale3D=(X=1,Y=1,Z=1))
    bEnabled=True
    Transform=@n20.Result
   )
   n25 = &RigUnit_BoneName(Bone=hair_4_r_tip)
   n26 = &RigVMFunction_MathVectorSub(A=&RigUnit_GetBoneTransform(Bone=@n25.Bone).Transform.Translation, B=@n20.Result.Translation) {template="Subtract::Execute(in A,in B,out Result)"}
   n27 = &RigVMFunction_MathVectorFromFloat(Value=&RigVMFunction_MathFloatMax(B=@n21.Result, A=&RigVMFunction_MathVectorLength(Value=@n26.Result) {template="Length::Execute(in Value,out Result)"}.Result) {template="Maximum::Execute(in A,in B,out Result)"}.Result) {template="Cast::Execute(in Value,out Result)"}
   n28 = &RigVMFunction_MathVectorUnit(Value=@n26.Result) {template="Unit::Execute(in Value,out Result)"}
   > &RigUnit_SetBoneTranslation(
    Space=GlobalSpace
    Weight=1
    bPropagateToChildren=True
    Bone=@n25.Bone
    Translation=&RigVMFunction_MathVectorAdd(
     B=@n20.Result.Translation
     A=&RigVMFunction_MathVectorMul(A=@n27.Result, B=@n28.Result) {template="Multiply::Execute(in A,in B,out Result)"}.Result
    ) {template="Add::Execute(in A,in B,out Result)"}.Result
   )
   n29 = &RigUnit_BeginExecution()
   n30 = &RigVMFunction_MathTransformMul(
    A=(Rotation=(X=0,Y=0,Z=0,W=1),Translation=(X=-2,Y=3,Z=0),Scale3D=(X=1,Y=1,Z=1))
    B=&RigUnit_GetBoneTransform(Bone=spine_05).Transform
   ) {template="Multiply::Execute(in A,in B,out Result)"}
   n31 = &RigVMFunction_MathFloatAbs(Value=20) {template="Absolute::Execute(in Value,out Result)"}
   n32 = &RigVMFunction_MathVectorSub(A=&RigUnit_GetBoneTransform(Bone=hair_5_r_end).Transform.Translation, B=@n30.Result.Translation) {template="Subtract::Execute(in A,in B,out Result)"}
   n33 = &RigVMFunction_MathVectorFromFloat(Value=&RigVMFunction_MathFloatMax(B=@n31.Result, A=&RigVMFunction_MathVectorLength(Value=@n32.Result) {template="Length::Execute(in Value,out Result)"}.Result) {template="Maximum::Execute(in A,in B,out Result)"}.Result) {template="Cast::Execute(in Value,out Result)"}
   n34 = &RigVMFunction_MathVectorUnit(Value=@n32.Result) {template="Unit::Execute(in Value,out Result)"}
   &RigUnit_AimBone(
    Bone=hair_5_cst_r
    DebugSettings=(bEnabled=False,Scale=10,WorldOffset=(Rotation=(X=0,Y=0,Z=0,W=1),Translation=(X=0,Y=0,Z=0),Scale3D=(X=1,Y=1,Z=1)))
    Primary=(Weight=1,Axis=(X=1,Y=0,Z=0),Target=(X=1,Y=0,Z=0),Kind=Location,Space="None")
    Secondary=(Weight=0,Axis=(X=0,Y=0,Z=1),Target=(X=1,Y=0,Z=0),Kind=Direction,Space="None")
    Weight=1
    bPropagateToChildren=True
    ExecutePin=@n29.ExecutePin
    Primary.Target=&RigVMFunction_MathVectorAdd(
     B=@n30.Result.Translation
     A=&RigVMFunction_MathVectorMul(A=@n33.Result, B=@n34.Result) {template="Multiply::Execute(in A,in B,out Result)"}.Result
    ) {template="Add::Execute(in A,in B,out Result)"}.Result
   )
   > &RigUnit_DebugTransformMutable(
    Color=(R=1,G=0,B=0,A=1)
    DebugDrawSettings=(DepthPriority=SDPG_Foreground,Lifetime=-1)
    Mode=Axes
    Scale=10
    Space=None
    Thickness=0
    WorldOffset=(Rotation=(X=0,Y=0,Z=0,W=1),Translation=(X=0,Y=0,Z=0),Scale3D=(X=1,Y=1,Z=1))
    bEnabled=True
    Transform=@n30.Result
   )
   n35 = &RigUnit_BoneName(Bone=hair_5_r_tip)
   n36 = &RigVMFunction_MathVectorSub(A=&RigUnit_GetBoneTransform(Bone=@n35.Bone).Transform.Translation, B=@n30.Result.Translation) {template="Subtract::Execute(in A,in B,out Result)"}
   n37 = &RigVMFunction_MathVectorFromFloat(Value=&RigVMFunction_MathFloatMax(B=@n31.Result, A=&RigVMFunction_MathVectorLength(Value=@n36.Result) {template="Length::Execute(in Value,out Result)"}.Result) {template="Maximum::Execute(in A,in B,out Result)"}.Result) {template="Cast::Execute(in Value,out Result)"}
   n38 = &RigVMFunction_MathVectorUnit(Value=@n36.Result) {template="Unit::Execute(in Value,out Result)"}
   > &RigUnit_SetBoneTranslation(
    Space=GlobalSpace
    Weight=1
    bPropagateToChildren=True
    Bone=@n35.Bone
    Translation=&RigVMFunction_MathVectorAdd(
     B=@n30.Result.Translation
     A=&RigVMFunction_MathVectorMul(A=@n37.Result, B=@n38.Result) {template="Multiply::Execute(in A,in B,out Result)"}.Result
    ) {template="Add::Execute(in A,in B,out Result)"}.Result
   )
   n39 = &RigUnit_BeginExecution()
   n40 = &RigVMFunction_MathTransformMul(
    A=(Rotation=(X=0,Y=0,Z=0,W=1),Translation=(X=-27,Y=37,Z=0),Scale3D=(X=1,Y=1,Z=1))
    B=&RigUnit_GetBoneTransform(Bone=spine_05).Transform
   ) {template="Multiply::Execute(in A,in B,out Result)"}
   n41 = &RigVMFunction_MathFloatAbs(Value=60) {template="Absolute::Execute(in Value,out Result)"}
   n42 = &RigVMFunction_MathVectorSub(A=&RigUnit_GetBoneTransform(Bone=hair_1_c_end).Transform.Translation, B=@n40.Result.Translation) {template="Subtract::Execute(in A,in B,out Result)"}
   n43 = &RigVMFunction_MathVectorFromFloat(Value=&RigVMFunction_MathFloatMax(B=@n41.Result, A=&RigVMFunction_MathVectorLength(Value=@n42.Result) {template="Length::Execute(in Value,out Result)"}.Result) {template="Maximum::Execute(in A,in B,out Result)"}.Result) {template="Cast::Execute(in Value,out Result)"}
   n44 = &RigVMFunction_MathVectorUnit(Value=@n42.Result) {template="Unit::Execute(in Value,out Result)"}
   &RigUnit_AimBone(
    Bone=hair_1_cst_c
    DebugSettings=(bEnabled=False,Scale=10,WorldOffset=(Rotation=(X=0,Y=0,Z=0,W=1),Translation=(X=0,Y=0,Z=0),Scale3D=(X=1,Y=1,Z=1)))
    Primary=(Weight=1,Axis=(X=1,Y=0,Z=0),Target=(X=1,Y=0,Z=0),Kind=Location,Space="None")
    Secondary=(Weight=0,Axis=(X=0,Y=0,Z=1),Target=(X=1,Y=0,Z=0),Kind=Direction,Space="None")
    Weight=1
    bPropagateToChildren=True
    ExecutePin=@n39.ExecutePin
    Primary.Target=&RigVMFunction_MathVectorAdd(
     B=@n40.Result.Translation
     A=&RigVMFunction_MathVectorMul(A=@n43.Result, B=@n44.Result) {template="Multiply::Execute(in A,in B,out Result)"}.Result
    ) {template="Add::Execute(in A,in B,out Result)"}.Result
   )
   > &RigUnit_DebugTransformMutable(
    Color=(R=1,G=0,B=0,A=1)
    DebugDrawSettings=(DepthPriority=SDPG_Foreground,Lifetime=-1)
    Mode=Axes
    Scale=10
    Space=None
    Thickness=0
    WorldOffset=(Rotation=(X=0,Y=0,Z=0,W=1),Translation=(X=0,Y=0,Z=0),Scale3D=(X=1,Y=1,Z=1))
    bEnabled=True
    Transform=@n40.Result
   )
   n45 = &RigUnit_BeginExecution()
   n46 = &RigVMFunction_MathTransformMul(
    A=(Rotation=(X=0,Y=0,Z=0,W=1),Translation=(X=-27,Y=37,Z=0),Scale3D=(X=1,Y=1,Z=1))
    B=&RigUnit_GetBoneTransform(Bone=spine_05).Transform
   ) {template="Multiply::Execute(in A,in B,out Result)"}
   n47 = &RigVMFunction_MathFloatAbs(Value=60) {template="Absolute::Execute(in Value,out Result)"}
   n48 = &RigVMFunction_MathVectorSub(A=&RigUnit_GetBoneTransform(Bone=hair_2_c_end).Transform.Translation, B=@n46.Result.Translation) {template="Subtract::Execute(in A,in B,out Result)"}
   n49 = &RigVMFunction_MathVectorFromFloat(Value=&RigVMFunction_MathFloatMax(B=@n47.Result, A=&RigVMFunction_MathVectorLength(Value=@n48.Result) {template="Length::Execute(in Value,out Result)"}.Result) {template="Maximum::Execute(in A,in B,out Result)"}.Result) {template="Cast::Execute(in Value,out Result)"}
   n50 = &RigVMFunction_MathVectorUnit(Value=@n48.Result) {template="Unit::Execute(in Value,out Result)"}
   &RigUnit_AimBone(
    Bone=hair_2_cst_c
    DebugSettings=(bEnabled=False,Scale=10,WorldOffset=(Rotation=(X=0,Y=0,Z=0,W=1),Translation=(X=0,Y=0,Z=0),Scale3D=(X=1,Y=1,Z=1)))
    Primary=(Weight=1,Axis=(X=1,Y=0,Z=0),Target=(X=1,Y=0,Z=0),Kind=Location,Space="None")
    Secondary=(Weight=0,Axis=(X=0,Y=0,Z=1),Target=(X=1,Y=0,Z=0),Kind=Direction,Space="None")
    Weight=1
    bPropagateToChildren=True
    ExecutePin=@n45.ExecutePin
    Primary.Target=&RigVMFunction_MathVectorAdd(
     B=@n46.Result.Translation
     A=&RigVMFunction_MathVectorMul(A=@n49.Result, B=@n50.Result) {template="Multiply::Execute(in A,in B,out Result)"}.Result
    ) {template="Add::Execute(in A,in B,out Result)"}.Result
   )
   > &RigUnit_DebugTransformMutable(
    Color=(R=1,G=0,B=0,A=1)
    DebugDrawSettings=(DepthPriority=SDPG_Foreground,Lifetime=-1)
    Mode=Axes
    Scale=10
    Space=None
    Thickness=0
    WorldOffset=(Rotation=(X=0,Y=0,Z=0,W=1),Translation=(X=0,Y=0,Z=0),Scale3D=(X=1,Y=1,Z=1))
    bEnabled=True
    Transform=@n46.Result
   )
   n51 = &RigUnit_BoneName(Bone=hair_2_c_tip)
   n52 = &RigVMFunction_MathVectorSub(A=&RigUnit_GetBoneTransform(Bone=@n51.Bone).Transform.Translation, B=@n46.Result.Translation) {template="Subtract::Execute(in A,in B,out Result)"}
   n53 = &RigVMFunction_MathVectorFromFloat(Value=&RigVMFunction_MathFloatMax(B=@n47.Result, A=&RigVMFunction_MathVectorLength(Value=@n52.Result) {template="Length::Execute(in Value,out Result)"}.Result) {template="Maximum::Execute(in A,in B,out Result)"}.Result) {template="Cast::Execute(in Value,out Result)"}
   n54 = &RigVMFunction_MathVectorUnit(Value=@n52.Result) {template="Unit::Execute(in Value,out Result)"}
   > &RigUnit_SetBoneTranslation(
    Space=GlobalSpace
    Weight=1
    bPropagateToChildren=True
    Bone=@n51.Bone
    Translation=&RigVMFunction_MathVectorAdd(
     B=@n46.Result.Translation
     A=&RigVMFunction_MathVectorMul(A=@n53.Result, B=@n54.Result) {template="Multiply::Execute(in A,in B,out Result)"}.Result
    ) {template="Add::Execute(in A,in B,out Result)"}.Result
   )
   n55 = &RigUnit_BeginExecution()
   n56 = &RigVMFunction_MathTransformMul(
    A=(Rotation=(X=0,Y=0,Z=0,W=1),Translation=(X=-2,Y=3,Z=0),Scale3D=(X=1,Y=1,Z=1))
    B=&RigUnit_GetBoneTransform(Bone=spine_05).Transform
   ) {template="Multiply::Execute(in A,in B,out Result)"}
   n57 = &RigVMFunction_MathFloatAbs(Value=20) {template="Absolute::Execute(in Value,out Result)"}
   n58 = &RigVMFunction_MathVectorSub(A=&RigUnit_GetBoneTransform(Bone=hair_3_c_end).Transform.Translation, B=@n56.Result.Translation) {template="Subtract::Execute(in A,in B,out Result)"}
   n59 = &RigVMFunction_MathVectorFromFloat(Value=&RigVMFunction_MathFloatMax(B=@n57.Result, A=&RigVMFunction_MathVectorLength(Value=@n58.Result) {template="Length::Execute(in Value,out Result)"}.Result) {template="Maximum::Execute(in A,in B,out Result)"}.Result) {template="Cast::Execute(in Value,out Result)"}
   n60 = &RigVMFunction_MathVectorUnit(Value=@n58.Result) {template="Unit::Execute(in Value,out Result)"}
   &RigUnit_AimBone(
    Bone=hair_3_cst_c
    DebugSettings=(bEnabled=False,Scale=10,WorldOffset=(Rotation=(X=0,Y=0,Z=0,W=1),Translation=(X=0,Y=0,Z=0),Scale3D=(X=1,Y=1,Z=1)))
    Primary=(Weight=1,Axis=(X=1,Y=0,Z=0),Target=(X=1,Y=0,Z=0),Kind=Location,Space="None")
    Secondary=(Weight=0,Axis=(X=0,Y=0,Z=1),Target=(X=1,Y=0,Z=0),Kind=Direction,Space="None")
    Weight=1
    bPropagateToChildren=True
    ExecutePin=@n55.ExecutePin
    Primary.Target=&RigVMFunction_MathVectorAdd(
     B=@n56.Result.Translation
     A=&RigVMFunction_MathVectorMul(A=@n59.Result, B=@n60.Result) {template="Multiply::Execute(in A,in B,out Result)"}.Result
    ) {template="Add::Execute(in A,in B,out Result)"}.Result
   )
   > &RigUnit_DebugTransformMutable(
    Color=(R=1,G=0,B=0,A=1)
    DebugDrawSettings=(DepthPriority=SDPG_Foreground,Lifetime=-1)
    Mode=Axes
    Scale=10
    Space=None
    Thickness=0
    WorldOffset=(Rotation=(X=0,Y=0,Z=0,W=1),Translation=(X=0,Y=0,Z=0),Scale3D=(X=1,Y=1,Z=1))
    bEnabled=True
    Transform=@n56.Result
   )
   n61 = &RigUnit_BoneName(Bone=hair_3_c_tip)
   n62 = &RigVMFunction_MathVectorSub(A=&RigUnit_GetBoneTransform(Bone=@n61.Bone).Transform.Translation, B=@n56.Result.Translation) {template="Subtract::Execute(in A,in B,out Result)"}
   n63 = &RigVMFunction_MathVectorFromFloat(Value=&RigVMFunction_MathFloatMax(B=@n57.Result, A=&RigVMFunction_MathVectorLength(Value=@n62.Result) {template="Length::Execute(in Value,out Result)"}.Result) {template="Maximum::Execute(in A,in B,out Result)"}.Result) {template="Cast::Execute(in Value,out Result)"}
   n64 = &RigVMFunction_MathVectorUnit(Value=@n62.Result) {template="Unit::Execute(in Value,out Result)"}
   > &RigUnit_SetBoneTranslation(
    Space=GlobalSpace
    Weight=1
    bPropagateToChildren=True
    Bone=@n61.Bone
    Translation=&RigVMFunction_MathVectorAdd(
     B=@n56.Result.Translation
     A=&RigVMFunction_MathVectorMul(A=@n63.Result, B=@n64.Result) {template="Multiply::Execute(in A,in B,out Result)"}.Result
    ) {template="Add::Execute(in A,in B,out Result)"}.Result
   )
   n65 = &RigUnit_BeginExecution()
   n66 = &RigVMFunction_MathTransformMul(
    A=(Rotation=(X=0,Y=0,Z=0,W=1),Translation=(X=-2,Y=3,Z=0),Scale3D=(X=1,Y=1,Z=1))
    B=&RigUnit_GetBoneTransform(Bone=spine_05).Transform
   ) {template="Multiply::Execute(in A,in B,out Result)"}
   n67 = &RigVMFunction_MathFloatAbs(Value=20) {template="Absolute::Execute(in Value,out Result)"}
   n68 = &RigVMFunction_MathVectorSub(A=&RigUnit_GetBoneTransform(Bone=hair_5_l_end).Transform.Translation, B=@n66.Result.Translation) {template="Subtract::Execute(in A,in B,out Result)"}
   n69 = &RigVMFunction_MathVectorFromFloat(Value=&RigVMFunction_MathFloatMax(B=@n67.Result, A=&RigVMFunction_MathVectorLength(Value=@n68.Result) {template="Length::Execute(in Value,out Result)"}.Result) {template="Maximum::Execute(in A,in B,out Result)"}.Result) {template="Cast::Execute(in Value,out Result)"}
   n70 = &RigVMFunction_MathVectorUnit(Value=@n68.Result) {template="Unit::Execute(in Value,out Result)"}
   &RigUnit_AimBone(
    Bone=hair_5_cst_l
    DebugSettings=(bEnabled=False,Scale=10,WorldOffset=(Rotation=(X=0,Y=0,Z=0,W=1),Translation=(X=0,Y=0,Z=0),Scale3D=(X=1,Y=1,Z=1)))
    Primary=(Weight=1,Axis=(X=1,Y=0,Z=0),Target=(X=1,Y=0,Z=0),Kind=Location,Space="None")
    Secondary=(Weight=0,Axis=(X=0,Y=0,Z=1),Target=(X=1,Y=0,Z=0),Kind=Direction,Space="None")
    Weight=1
    bPropagateToChildren=True
    ExecutePin=@n65.ExecutePin
    Primary.Target=&RigVMFunction_MathVectorAdd(
     B=@n66.Result.Translation
     A=&RigVMFunction_MathVectorMul(A=@n69.Result, B=@n70.Result) {template="Multiply::Execute(in A,in B,out Result)"}.Result
    ) {template="Add::Execute(in A,in B,out Result)"}.Result
   )
   > &RigUnit_DebugTransformMutable(
    Color=(R=1,G=0,B=0,A=1)
    DebugDrawSettings=(DepthPriority=SDPG_Foreground,Lifetime=-1)
    Mode=Axes
    Scale=10
    Space=None
    Thickness=0
    WorldOffset=(Rotation=(X=0,Y=0,Z=0,W=1),Translation=(X=0,Y=0,Z=0),Scale3D=(X=1,Y=1,Z=1))
    bEnabled=True
    Transform=@n66.Result
   )
   n71 = &RigUnit_BoneName(Bone=hair_5_l_tip)
   n72 = &RigVMFunction_MathVectorSub(A=&RigUnit_GetBoneTransform(Bone=@n71.Bone).Transform.Translation, B=@n66.Result.Translation) {template="Subtract::Execute(in A,in B,out Result)"}
   n73 = &RigVMFunction_MathVectorFromFloat(Value=&RigVMFunction_MathFloatMax(B=@n67.Result, A=&RigVMFunction_MathVectorLength(Value=@n72.Result) {template="Length::Execute(in Value,out Result)"}.Result) {template="Maximum::Execute(in A,in B,out Result)"}.Result) {template="Cast::Execute(in Value,out Result)"}
   n74 = &RigVMFunction_MathVectorUnit(Value=@n72.Result) {template="Unit::Execute(in Value,out Result)"}
   > &RigUnit_SetBoneTranslation(
    Space=GlobalSpace
    Weight=1
    bPropagateToChildren=True
    Bone=@n71.Bone
    Translation=&RigVMFunction_MathVectorAdd(
     B=@n66.Result.Translation
     A=&RigVMFunction_MathVectorMul(A=@n73.Result, B=@n74.Result) {template="Multiply::Execute(in A,in B,out Result)"}.Result
    ) {template="Add::Execute(in A,in B,out Result)"}.Result
   )
   n75 = &RigVMFunction_MathTransformMul(
    A=(Rotation=(X=0,Y=0,Z=0,W=1),Translation=(X=13,Y=3,Z=-29),Scale3D=(X=1,Y=1,Z=1))
    B=&RigUnit_GetBoneTransform(Bone=upperarm_l).Transform
   ) {template="Multiply::Execute(in A,in B,out Result)"}
   n76 = &RigVMFunction_MathFloatAbs(Value=38) {template="Absolute::Execute(in Value,out Result)"}
   n77 = &RigVMFunction_MathVectorSub(A=&RigUnit_GetBoneTransform(Bone=hair_5_l_end).Transform.Translation, B=@n75.Result.Translation) {template="Subtract::Execute(in A,in B,out Result)"}
   n78 = &RigVMFunction_MathVectorFromFloat(Value=&RigVMFunction_MathFloatMax(B=@n76.Result, A=&RigVMFunction_MathVectorLength(Value=@n77.Result) {template="Length::Execute(in Value,out Result)"}.Result) {template="Maximum::Execute(in A,in B,out Result)"}.Result) {template="Cast::Execute(in Value,out Result)"}
   n79 = &RigVMFunction_MathVectorUnit(Value=@n77.Result) {template="Unit::Execute(in Value,out Result)"}
   > &RigUnit_AimBone(
    Bone=hair_5_cst_l
    DebugSettings=(bEnabled=False,Scale=10,WorldOffset=(Rotation=(X=0,Y=0,Z=0,W=1),Translation=(X=0,Y=0,Z=0),Scale3D=(X=1,Y=1,Z=1)))
    Primary=(Weight=1,Axis=(X=1,Y=0,Z=0),Target=(X=1,Y=0,Z=0),Kind=Location,Space="None")
    Secondary=(Weight=0,Axis=(X=0,Y=0,Z=1),Target=(X=1,Y=0,Z=0),Kind=Direction,Space="None")
    Weight=1
    bPropagateToChildren=True
    Primary.Target=&RigVMFunction_MathVectorAdd(
     B=@n75.Result.Translation
     A=&RigVMFunction_MathVectorMul(A=@n78.Result, B=@n79.Result) {template="Multiply::Execute(in A,in B,out Result)"}.Result
    ) {template="Add::Execute(in A,in B,out Result)"}.Result
   )
   > &RigUnit_DebugTransformMutable(
    Color=(R=1,G=0,B=0,A=1)
    DebugDrawSettings=(DepthPriority=SDPG_Foreground,Lifetime=-1)
    Mode=Axes
    Scale=10
    Space=None
    Thickness=0
    WorldOffset=(Rotation=(X=0,Y=0,Z=0,W=1),Translation=(X=0,Y=0,Z=0),Scale3D=(X=1,Y=1,Z=1))
    bEnabled=True
    Transform=@n75.Result
   )
   n80 = &RigUnit_BoneName(Bone=hair_5_l_tip)
   n81 = &RigVMFunction_MathVectorSub(A=&RigUnit_GetBoneTransform(Bone=@n80.Bone).Transform.Translation, B=@n75.Result.Translation) {template="Subtract::Execute(in A,in B,out Result)"}
   n82 = &RigVMFunction_MathVectorFromFloat(Value=&RigVMFunction_MathFloatMax(B=@n76.Result, A=&RigVMFunction_MathVectorLength(Value=@n81.Result) {template="Length::Execute(in Value,out Result)"}.Result) {template="Maximum::Execute(in A,in B,out Result)"}.Result) {template="Cast::Execute(in Value,out Result)"}
   n83 = &RigVMFunction_MathVectorUnit(Value=@n81.Result) {template="Unit::Execute(in Value,out Result)"}
   > &RigUnit_SetBoneTranslation(
    Space=GlobalSpace
    Weight=1
    bPropagateToChildren=True
    Bone=@n80.Bone
    Translation=&RigVMFunction_MathVectorAdd(
     B=@n75.Result.Translation
     A=&RigVMFunction_MathVectorMul(A=@n82.Result, B=@n83.Result) {template="Multiply::Execute(in A,in B,out Result)"}.Result
    ) {template="Add::Execute(in A,in B,out Result)"}.Result
   )
   &RigUnit_BeginExecution()
   &RigUnit_BeginExecution()
   &RigUnit_BeginExecution()
   n84 = &RigUnit_BoneName(Bone=hair_1_c_tip)
   n85 = &RigVMFunction_MathVectorSub(A=&RigUnit_GetBoneTransform(Bone=@n84.Bone).Transform.Translation, B=@n40.Result.Translation) {template="Subtract::Execute(in A,in B,out Result)"}
   n86 = &RigVMFunction_MathVectorFromFloat(Value=&RigVMFunction_MathFloatMax(B=@n41.Result, A=&RigVMFunction_MathVectorLength(Value=@n85.Result) {template="Length::Execute(in Value,out Result)"}.Result) {template="Maximum::Execute(in A,in B,out Result)"}.Result) {template="Cast::Execute(in Value,out Result)"}
   n87 = &RigVMFunction_MathVectorUnit(Value=@n85.Result) {template="Unit::Execute(in Value,out Result)"}
   &RigUnit_SetBoneTranslation(
    ExecutePin=()
    Space=GlobalSpace
    Weight=1
    bPropagateToChildren=True
    Bone=@n84.Bone
    Translation=&RigVMFunction_MathVectorAdd(
     B=@n40.Result.Translation
     A=&RigVMFunction_MathVectorMul(A=@n86.Result, B=@n87.Result) {template="Multiply::Execute(in A,in B,out Result)"}.Result
    ) {template="Add::Execute(in A,in B,out Result)"}.Result
   )
  ]}
]
hierarchy [
 root {loc=<0 0 0>, rot=<0 0 0 1>}
 pelvis {parent=root, loc=<0 -0.5301 98.0571>, rot=<0.0127 0.707 -0.0127 -0.707>}
 spine_01 {parent=pelvis, loc=<6.1208 -0.9328 0>, rot=<0 0 0.034 -0.9994>}
 spine_02 {parent=spine_01, loc=<6.5996 -0.2095 0>, rot=<0 0 0.0238 -0.9997>}
 spine_03 {parent=spine_02, loc=<6.4914 -0.1237 0>, rot=<0 0 0.0149 -0.9999>}
 spine_04 {parent=spine_03, loc=<6.9598 0.0794 0>, rot=<0 0 -0.0132 -0.9999>}
 spine_05 {parent=spine_04, loc=<6.9987 0 0>, rot=<0 0 -0.0015 -1>}
 clavicle_l {parent=spine_05, loc=<5.4831 -0.1053 -3.2946>, rot=<0.0465 -0.7289 -0.009 -0.6829>}
 upperarm_l {parent=clavicle_l, loc=<9.8658 -1.2136 -0.4105>, rot=<0.0164 -0.3245 -0.0269 -0.9454>}
 clavicle_r {parent=spine_05, loc=<5.4831 -0.1053 3.2946>, rot=<0.7289 0.0465 -0.6829 0.009>}
 upperarm_r {parent=clavicle_r, loc=<-9.8658 1.2136 0.4104>, rot=<0.0164 -0.3245 -0.0269 -0.9454>}
 neck_01 {parent=spine_05, loc=<11.3138 0.1529 0>, rot=<0 0 -0.2003 -0.9797>}
 neck_02 {parent=neck_01, loc=<4.5627 0 0>, rot=<0 0 0.0504 -0.9987>}
 head {parent=neck_02, loc=<5.3914 0 0>, rot=<0 0 0.115 -0.9934>}
 hair_6_r_zero {parent=head, loc=<4.14 -3.9872 7.1226>, rot=<0.193 -0.8412 0.492 0.114>}
 hair_6_r {parent=hair_6_r_zero, loc=<0 0 0>, rot=<0 0 0 -1>}
 hair_6_cst_r {parent=hair_6_r, loc=<0 0 0>, rot=<0 0 0 -1>}
 hair_6_r_end {parent=hair_6_cst_r, loc=<4.7184 0 0>, rot=<0.0145 -0.0581 -0.2706 -0.9608>}
 hair_6_r_tip {parent=hair_6_r_end, loc=<4.4442 0 0>, rot=<0 0 0 1>}
 hair_5_l_zero {parent=head, loc=<4.2596 -4.6185 -6.7217>, rot=<0.0488 -0.7409 -0.668 -0.0499>}
 hair_5_l {parent=hair_5_l_zero, loc=<0 0 0>, rot=<0 0 0 -1>}
 hair_5_cst_l {parent=hair_5_l, loc=<0 0 0>, rot=<0 0 0 -1>}
 hair_5_l_end {parent=hair_5_cst_l, loc=<12.1901 0 0>, rot=<0.0028 0.0293 0.5675 -0.8228>}
 hair_5_l_tip {parent=hair_5_l_end, loc=<3.8281 0 0>, rot=<0 0 0 1>}
 hair_6_l_zero {parent=hair_5_cst_l, loc=<-0.0174 0.1706 1.5007>, rot=<-0.0347 -0.0855 0.134 -0.9867>}
 hair_6_l {parent=hair_6_l_zero, loc=<0 0 0>, rot=<0 0 0 -1>}
 hair_6_cst_l {parent=hair_6_l, loc=<0 0 0>, rot=<0 0 0 -1>}
 hair_6_l_end {parent=hair_6_cst_l, loc=<6.7174 0 0>, rot=<-0.0014 -0.0454 0.1619 -0.9858>}
 hair_6_l_tip {parent=hair_6_l_end, loc=<2.5117 0 0>, rot=<0 0 0 1>}
 hair_7_l_zero {parent=hair_5_cst_l, loc=<0.0779 -0.1356 -1.4415>, rot=<-0.2993 -0.0749 0.1361 -0.9414>}
 hair_7_l {parent=hair_7_l_zero, loc=<0 0 0>, rot=<0 0 0 1>}
 hair_7_cst_l {parent=hair_7_l, loc=<0 0 0>, rot=<0 0 0 1>}
 hair_7_l_end {parent=hair_7_cst_l, loc=<4.298 0 0>, rot=<0.0109 -0.0723 -0.2199 -0.9728>}
 hair_7_l_tip {parent=hair_7_l_end, loc=<2.4482 0 0>, rot=<0 0 0 1>}
 hair_neck_02 {parent=neck_02, loc=<0 0 0>, rot=<0 0 0 -1>}
 hair_2_c_zero {parent=hair_neck_02, loc=<8.6477 -8.9969 -3.5529>, rot=<-0.0533 -0.9275 -0.362 -0.0763>}
 hair_2_c {parent=hair_2_c_zero, loc=<0 0 0>, rot=<0 0 0 1>}
 hair_2_cst_c {parent=hair_2_c, loc=<0 0 0>, rot=<0 0 0 1>}
 hair_2_c_end {parent=hair_2_cst_c, loc=<13.3222 0 0>, rot=<0.008 -0.002 0.5328 -0.8462>}
 hair_2_c_tip {parent=hair_2_c_end, loc=<4.661 0 0>, rot=<0 0 0 -1>}
 hair_1_c_zero {parent=hair_neck_02, loc=<8.5336 -9.5252 1.9138>, rot=<-0.0693 -0.9853 0.1301 -0.086>}
 hair_1_c {parent=hair_1_c_zero, loc=<0 0 0>, rot=<0 0 0 1>}
 hair_1_cst_c {parent=hair_1_c, loc=<0 0 0>, rot=<0 0 0 1>}
 hair_1_c_end {parent=hair_1_cst_c, loc=<13.3747 0 0>, rot=<0.3419 0.2351 0.554 -0.7217>}
 hair_1_c_tip {parent=hair_1_c_end, loc=<4.5339 0 0>, rot=<0 0 0 -1>}
 hair_3_c_zero {parent=hair_neck_02, loc=<8.6897 -7.3158 4.7867>, rot=<-0.0493 -0.8615 0.5023 -0.0556>}
 hair_3_c {parent=hair_3_c_zero, loc=<0 0 0>, rot=<0 0 0 1>}
 hair_3_cst_c {parent=hair_3_c, loc=<0 0 0>, rot=<0 0 0 1>}
 hair_3_c_end {parent=hair_3_cst_c, loc=<14.4287 0 0>, rot=<0.4531 0.4782 0.0448 -0.751>}
 hair_3_c_tip {parent=hair_3_c_end, loc=<4.961 0 0>, rot=<0 0 0 1>}
 hair_5_r_zero {parent=hair_neck_02, loc=<8.6236 -6.2214 6.1539>, rot=<-0.0471 -0.8567 0.5092 0.0679>}
 hair_5_r {parent=hair_5_r_zero, loc=<0 0 0>, rot=<0 0 0 -1>}
 hair_5_cst_r {parent=hair_5_r, loc=<0 0 0>, rot=<0 0 0 -1>}
 hair_5_r_end {parent=hair_5_cst_r, loc=<12.1043 0 0>, rot=<-0.2624 -0.1992 0.5709 -0.752>}
 hair_5_r_tip {parent=hair_5_r_end, loc=<5.5929 0 0>, rot=<0 0 0 -1>}
 hair_group_r_zero {parent=hair_neck_02, loc=<8.3289 -0.6997 6.4327>, rot=<-0.0784 -0.7027 0.7027 0.0784>}
 hair_group_r {parent=hair_group_r_zero, loc=<0 0 0>, rot=<0 0 0 -1>}
 hair_2_r_zero {parent=hair_group_r, loc=<0.8094 0.663 1.9379>, rot=<-0.0268 -0.145 -0.0156 -0.989>}
 hair_2_r {parent=hair_2_r_zero, loc=<0 0 0>, rot=<0 0 0 -1>}
 hair_2_cst_r {parent=hair_2_r, loc=<0 0 0>, rot=<0 0 0 -1>}
 hair_2_r_end {parent=hair_2_cst_r, loc=<13.134 0 0>, rot=<0.0425 -0.1839 -0.304 -0.9338>}
 hair_2_r_tip {parent=hair_2_r_end, loc=<4.8629 0 0>, rot=<0 0 0 1>}
 hair_3_r_zero {parent=hair_group_r, loc=<3.2352 0.0487 0.4645>, rot=<0.0388 -0.1002 0.0638 -0.9922>}
 hair_3_r {parent=hair_3_r_zero, loc=<0 0 0>, rot=<0 0 0 1>}
 hair_3_cst_r {parent=hair_3_r, loc=<0 0 0>, rot=<0 0 0 1>}
 hair_3_r_end {parent=hair_3_cst_r, loc=<7.7782 0 0>, rot=<0.0309 -0.1977 -0.1834 -0.9625>}
 hair_3_r_tip {parent=hair_3_r_end, loc=<4.6592 0 0>, rot=<0 0 0 -1>}
 hair_1_r_zero {parent=hair_group_r, loc=<3.0267 0.4293 -0.9057>, rot=<0.1135 -0.1109 0.0211 -0.9871>}
 hair_1_r {parent=hair_1_r_zero, loc=<0 0 0>, rot=<0 0 0 -1>}
 hair_1_cst_r {parent=hair_1_r, loc=<0 0 0>, rot=<0 0 0 -1>}
 hair_1_r_end {parent=hair_1_cst_r, loc=<4.7289 0 0>, rot=<-0.1422 -0.1614 -0.2697 -0.9386>}
 hair_1_r_tip {parent=hair_1_r_end, loc=<4.3398 0 0>, rot=<0 0 0 -1>}
 hair_4_r_zero {parent=hair_group_r, loc=<3.9829 -1.1675 1.7191>, rot=<0.1131 -0.0459 -0.001 -0.9925>}
 hair_4_r {parent=hair_4_r_zero, loc=<0 0 0>, rot=<0 0 0 -1>}
 hair_4_cst_r {parent=hair_4_r, loc=<0 0 0>, rot=<0 0 0 -1>}
 hair_4_r_end {parent=hair_4_cst_r, loc=<7.1308 0 0>, rot=<-0.0162 -0.2613 0.3195 -0.9107>}
 hair_4_r_tip {parent=hair_4_r_end, loc=<5.4645 0 0>, rot=<0 0 0 -1>}
 hair_group_l_zero {parent=hair_neck_02, loc=<8.3289 -0.6997 -6.4327>, rot=<0.0784 0.7027 0.7027 0.0784>}
 hair_group_l {parent=hair_group_l_zero, loc=<0 0 0>, rot=<0 0 0 -1>}
 hair_1_l_zero {parent=hair_group_l, loc=<3.0267 0.4292 0.9057>, rot=<-0.1135 0.1109 0.0211 -0.9871>}
 hair_1_l {parent=hair_1_l_zero, loc=<0 0 0>, rot=<0 0 0 -1>}
 hair_1_cst_l {parent=hair_1_l, loc=<0 0 0>, rot=<0 0 0 -1>}
 hair_1_l_end {parent=hair_1_cst_l, loc=<4.7289 0 0>, rot=<0.1422 0.1614 -0.2697 -0.9386>}
 hair_1_l_tip {parent=hair_1_l_end, loc=<4.3398 0 0>, rot=<0 0 0 -1>}
 hair_3_l_zero {parent=hair_group_l, loc=<3.6552 -1.2119 -1.2345>, rot=<-0.0796 0.1278 -0.0007 -0.9886>}
 hair_3_l {parent=hair_3_l_zero, loc=<-0.2355 1.3747 0.6425>, rot=<0.0489 -0.0227 0.0608 -0.9967>}
 hair_3_cst_l {parent=hair_3_l, loc=<0 0 0>, rot=<0 0 0 1>}
 hair_3_l_end {parent=hair_3_cst_l, loc=<7.7782 0 0>, rot=<-0.0309 0.1977 -0.1834 -0.9625>}
 hair_3_l_tip {parent=hair_3_l_end, loc=<4.6592 0 0>, rot=<0 0 0 -1>}
 hair_4_l_zero {parent=hair_group_l, loc=<3.9829 -1.1675 -1.7191>, rot=<-0.1131 0.0459 -0.001 -0.9925>}
 hair_4_l {parent=hair_4_l_zero, loc=<0 0 0>, rot=<0 0 0 1>}
 hair_4_cst_l {parent=hair_4_l, loc=<0 0 0>, rot=<0 0 0 1>}
 hair_4_l_end {parent=hair_4_cst_l, loc=<7.1308 0 0>, rot=<0.0162 0.2613 0.3195 -0.9107>}
 hair_4_l_tip {parent=hair_4_l_end, loc=<5.4645 0 0>, rot=<0 0 0 -1>}
 hair_2_l_zero {parent=hair_group_l, loc=<0.8094 0.663 -1.9379>, rot=<0.0268 0.145 -0.0156 -0.989>}
 hair_2_l {parent=hair_2_l_zero, loc=<0 0 0>, rot=<0 0 0 1>}
 hair_2_cst_l {parent=hair_2_l, loc=<0 0 0>, rot=<0 0 0 1>}
 hair_2_l_end {parent=hair_2_cst_l, loc=<13.134 0 0>, rot=<-0.0425 0.1839 -0.304 -0.9338>}
 hair_2_l_tip {parent=hair_2_l_end, loc=<4.8629 0 0>, rot=<0 0 0 1>}
]
