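# Post-process style hair animation for the F_MED_Pink_Trooper face accessory skeleton, adding secondary motion to the character's hair bones.
# Spring-driven AnimDynamics nodes swing hair and tip bones from cached 'cst_pose', using spherical limits at neck, clavicles and head.
# Weighted rotation constraints blend neighboring strands into intermediate bones, and a Control Rig finishes the output pose.
# A separate branch from 'orig_cst' positions hair target bones between spine and clavicles, then offsets them.
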AnimBlueprint /BRCosmetics/Characters/Player/Female/Medium/Bodies/F_MED_Pink_Trooper/Meshes/Parts/F_MED_Pink_Trooper_FaceAcc_AnimBP
target_skeleton=&/F_MED_Pink_Trooper_FaceAcc_Skeleton
graphs [
 AnimGraph {nodes [
   n1 = Anim Dynamics - Bone: hair_group_r(ComponentPose=Local To Component(LocalPose=Use cached pose 'cst_pose'()), Alpha=1, ExternalForce="0,0,0") {
    angularDampingOverride=1
    angularSpringConstant=1500
    bAngularSpring=true
    bGravityOverrideInSimSpace=true
    bOverrideAngularDamping=true
    bUseSphericalLimits=true
    gravityScale=0.35
    simulationSpace=RootRelative
    alphaBoolBlend {blendOption=Linear}
    alphaScaleBiasClamp {scale=0.7}
    boundBone {boneName=hair_group_r}
    componentLinearAccScale {z=0.46}
    componentLinearVelScale {z=0.46}
    physicsBodyDefinitions [{
      collisionType=CustomSphere
      sphereCollisionRadius=2
      boundBone {boneName=hair_group_r}
      boxExtents=<10 30 30>
      constraintSetup {angularLimitsMax=<0 45 20>, angularLimitsMin=<0 -45 -80>, angularTarget=<1 0 0>}
      localJointOffset=<7 0 0>
     }]
    sphericalLimits [
     {limitRadius=10, drivingBone {boneName=neck_02}, sphereLocalOffset=<-10 0 0>}
     {limitRadius=10, drivingBone {boneName=clavicle_l}, sphereLocalOffset=<10 0 -3>}
     {limitRadius=10, drivingBone {boneName=clavicle_r}, sphereLocalOffset=<-10 0 3>}
     {limitRadius=6, drivingBone {boneName=head}, sphereLocalOffset=<0 0 0>}
     {limitRadius=0, drivingBone {boneName=None}, sphereLocalOffset=<0 0 0>}
    ]
   }
   n2 = Anim Dynamics - Bone: hair_5_r(ComponentPose=@n1, Alpha=1, ExternalForce="0,0,0") {
    angularDampingOverride=1
    angularSpringConstant=1500
    bAngularSpring=true
    bGravityOverrideInSimSpace=true
    bOverrideAngularDamping=true
    bUseSphericalLimits=true
    gravityScale=0.35
    simulationSpace=RootRelative
    alphaBoolBlend {blendOption=Linear}
    boundBone {boneName=hair_5_r}
    componentLinearAccScale {z=0.46}
    componentLinearVelScale {z=0.46}
    physicsBodyDefinitions [{
      collisionType=CustomSphere
      sphereCollisionRadius=2
      boundBone {boneName=hair_5_r}
      boxExtents=<10 30 30>
      constraintSetup {angularLimitsMax=<0 20 20>, angularLimitsMin=<0 -20 -80>, angularTarget=<1 0 0>}
      localJointOffset=<5 0 0>
     }]
    sphericalLimits [
     {limitRadius=10, drivingBone {boneName=neck_02}, sphereLocalOffset=<-10 0 0>}
     {limitRadius=10, drivingBone {boneName=clavicle_l}, sphereLocalOffset=<10 0 -3>}
     {limitRadius=14, drivingBone {boneName=clavicle_r}, sphereLocalOffset=<-10 0 3>}
     {limitRadius=7, drivingBone {boneName=None}, sphereLocalOffset=<0 0 0>}
     {limitRadius=10, drivingBone {boneName=None}, sphereLocalOffset=<-10 10 0>}
     {limitRadius=10, drivingBone {boneName=None}, sphereLocalOffset=<-10 -10 0>}
    ]
   }
   n3 = Anim Dynamics - Bone: hair_6_r(ComponentPose=@n2, Alpha=1, ExternalForce="0,0,0") {
    angularDampingOverride=1
    angularSpringConstant=1500
    bAngularSpring=true
    bGravityOverrideInSimSpace=true
    bOverrideAngularDamping=true
    bUseSphericalLimits=true
    gravityScale=0.35
    simulationSpace=RootRelative
    alphaBoolBlend {blendOption=Linear}
    alphaScaleBiasClamp {scale=0.7}
    boundBone {boneName=hair_6_r}
    componentLinearAccScale {z=0.46}
    componentLinearVelScale {z=0.46}
    physicsBodyDefinitions [{
      collisionType=CustomSphere
      sphereCollisionRadius=2
      boundBone {boneName=hair_6_r}
      boxExtents=<10 30 30>
      constraintSetup {angularLimitsMax=<0 45 20>, angularLimitsMin=<0 -45 -80>, angularTarget=<1 0 0>}
      localJointOffset=<10 0 0>
     }]
    sphericalLimits [
     {limitRadius=10, drivingBone {boneName=neck_02}, sphereLocalOffset=<-10 0 0>}
     {limitRadius=10, drivingBone {boneName=clavicle_l}, sphereLocalOffset=<10 0 -3>}
     {limitRadius=10, drivingBone {boneName=clavicle_r}, sphereLocalOffset=<-10 0 3>}
     {limitRadius=10, drivingBone {boneName=None}, sphereLocalOffset=<0 0 0>}
     {limitRadius=9, drivingBone {boneName=None}, sphereLocalOffset=<0 0 0>}
    ]
   }
   n4 = Anim Dynamics - Bone: hair_group_l(ComponentPose=@n3, Alpha=1, ExternalForce="0,0,0") {
    angularDampingOverride=1
    angularSpringConstant=1500
    bAngularSpring=true
    bGravityOverrideInSimSpace=true
    bOverrideAngularDamping=true
    bUseSphericalLimits=true
    gravityScale=0.35
    simulationSpace=RootRelative
    alphaBoolBlend {blendOption=Linear}
    alphaScaleBiasClamp {scale=0.7}
    boundBone {boneName=hair_group_l}
    componentLinearAccScale {z=0.46}
    componentLinearVelScale {z=0.46}
    physicsBodyDefinitions [{
      collisionType=CustomSphere
      sphereCollisionRadius=2
      boundBone {boneName=hair_group_l}
      boxExtents=<10 30 30>
      constraintSetup {angularLimitsMax=<0 20 20>, angularLimitsMin=<0 -20 -80>, angularTarget=<1 0 0>}
      localJointOffset=<7 0 0>
     }]
    sphericalLimits [
     {limitRadius=10, drivingBone {boneName=neck_02}, sphereLocalOffset=<-10 0 0>}
     {limitRadius=10, drivingBone {boneName=clavicle_l}, sphereLocalOffset=<10 0 -3>}
     {limitRadius=10, drivingBone {boneName=clavicle_r}, sphereLocalOffset=<-10 0 3>}
     {limitRadius=6, drivingBone {boneName=head}, sphereLocalOffset=<0 0 0>}
    ]
   }
   n5 = Anim Dynamics - Bone: hair_5_l(ComponentPose=@n4, Alpha=1, ExternalForce="0,0,0") {
    angularDampingOverride=1
    angularSpringConstant=1500
    bAngularSpring=true
    bGravityOverrideInSimSpace=true
    bOverrideAngularDamping=true
    bUseSphericalLimits=true
    gravityScale=0.35
    simulationSpace=RootRelative
    alphaBoolBlend {blendOption=Linear}
    alphaScaleBiasClamp {scale=0.7}
    boundBone {boneName=hair_5_l}
    componentLinearAccScale {z=0.46}
    componentLinearVelScale {z=0.46}
    physicsBodyDefinitions [{
      collisionType=CustomSphere
      sphereCollisionRadius=2
      boundBone {boneName=hair_5_l}
      boxExtents=<10 30 30>
      constraintSetup {angularLimitsMax=<0 10 20>, angularLimitsMin=<0 -10 -80>, angularTarget=<1 0 0>}
      localJointOffset=<10 0 0>
     }]
    sphericalLimits [
     {limitRadius=10, drivingBone {boneName=neck_02}, sphereLocalOffset=<-10 0 0>}
     {limitRadius=10, drivingBone {boneName=clavicle_l}, sphereLocalOffset=<10 0 -3>}
     {limitRadius=10, drivingBone {boneName=clavicle_r}, sphereLocalOffset=<-10 0 3>}
     {limitRadius=7, drivingBone {boneName=head}, sphereLocalOffset=<0 0 0>}
    ]
   }
   n6 = Anim Dynamics - Bone: hair_1_c(ComponentPose=@n5, Alpha=1, ExternalForce="0,0,0") {
    angularDampingOverride=1
    angularSpringConstant=1500
    bAngularSpring=true
    bGravityOverrideInSimSpace=true
    bOverrideAngularDamping=true
    bUseSphericalLimits=true
    gravityScale=0.35
    simulationSpace=RootRelative
    alphaBoolBlend {blendOption=Linear}
    alphaScaleBiasClamp {scale=0.7}
    boundBone {boneName=hair_1_c}
    componentLinearAccScale {z=0.46}
    componentLinearVelScale {z=0.46}
    physicsBodyDefinitions [{
      collisionType=CustomSphere
      sphereCollisionRadius=2
      boundBone {boneName=hair_1_c}
      boxExtents=<10 30 30>
      constraintSetup {angularLimitsMax=<0 20 20>, angularLimitsMin=<0 -20 -80>, angularTarget=<1 0 0>}
      localJointOffset=<10 0 0>
     }]
    sphericalLimits [
     {limitRadius=10, drivingBone {boneName=neck_02}, sphereLocalOffset=<-10 0 0>}
     {limitRadius=10, drivingBone {boneName=clavicle_l}, sphereLocalOffset=<10 0 -3>}
     {limitRadius=10, drivingBone {boneName=clavicle_r}, sphereLocalOffset=<-10 0 3>}
     {limitRadius=8, drivingBone {boneName=head}, sphereLocalOffset=<0 0 0>}
    ]
   }
   n7 = Anim Dynamics - Bone: hair_2_c(ComponentPose=@n6, Alpha=1, ExternalForce="0,0,0") {
    angularDampingOverride=1
    angularSpringConstant=1500
    bAngularSpring=true
    bGravityOverrideInSimSpace=true
    bOverrideAngularDamping=true
    bUseSphericalLimits=true
    gravityScale=0.35
    simulationSpace=RootRelative
    alphaBoolBlend {blendOption=Linear}
    alphaScaleBiasClamp {scale=0.7}
    boundBone {boneName=hair_2_c}
    componentLinearAccScale {z=0.46}
    componentLinearVelScale {z=0.46}
    physicsBodyDefinitions [{
      collisionType=CustomSphere
      sphereCollisionRadius=2
      boundBone {boneName=hair_2_c}
      boxExtents=<10 30 30>
      constraintSetup {angularLimitsMax=<0 20 20>, angularLimitsMin=<0 -20 -80>, angularTarget=<1 0 0>}
      localJointOffset=<10 0 0>
     }]
    sphericalLimits [
     {limitRadius=10, drivingBone {boneName=neck_02}, sphereLocalOffset=<-10 0 0>}
     {limitRadius=10, drivingBone {boneName=clavicle_l}, sphereLocalOffset=<10 0 -3>}
     {limitRadius=10, drivingBone {boneName=clavicle_r}, sphereLocalOffset=<-10 0 3>}
     {limitRadius=8, drivingBone {boneName=head}, sphereLocalOffset=<0 0 0>}
    ]
   }
   n8 = Anim Dynamics - Bone: hair_3_c(ComponentPose=@n7, Alpha=1, ExternalForce="0,0,0") {
    angularDampingOverride=1
    angularSpringConstant=1500
    bAngularSpring=true
    bGravityOverrideInSimSpace=true
    bOverrideAngularDamping=true
    bUseSphericalLimits=true
    gravityScale=0.35
    simulationSpace=RootRelative
    alphaBoolBlend {blendOption=Linear}
    alphaScaleBiasClamp {scale=0.7}
    boundBone {boneName=hair_3_c}
    componentLinearAccScale {z=0.46}
    componentLinearVelScale {z=0.46}
    physicsBodyDefinitions [{
      collisionType=CustomSphere
      sphereCollisionRadius=2
      boundBone {boneName=hair_3_c}
      boxExtents=<10 30 30>
      constraintSetup {angularLimitsMax=<0 20 20>, angularLimitsMin=<0 -20 -80>, angularTarget=<1 0 0>}
      localJointOffset=<10 0 0>
     }]
    sphericalLimits [
     {limitRadius=10, drivingBone {boneName=neck_02}, sphereLocalOffset=<-10 0 0>}
     {limitRadius=10, drivingBone {boneName=clavicle_l}, sphereLocalOffset=<10 0 -3>}
     {limitRadius=12, drivingBone {boneName=clavicle_r}, sphereLocalOffset=<-10 0 3>}
     {limitRadius=8, drivingBone {boneName=head}, sphereLocalOffset=<0 0 0>}
    ]
   }
   n9 = "Constraint - hair_4_cst_r (RRRRRR)"!(
    ComponentPose=@n8
    Alpha=1
    ConstraintWeights_0=0.7
    ConstraintWeights_1:float=0.3
    ConstraintWeights_2:float=0
    ConstraintWeights_3:float=0
    ConstraintWeights_4:float=0
    ConstraintWeights_5:float=0
   ) -> (Pose) {
    alphaBoolBlend {blendOption=Linear}
    boneToModify {boneName=hair_4_cst_r}
    constraintSetup [
     {transformType=Rotation, targetBone {boneName=hair_4_r}}
     {transformType=Rotation, targetBone {boneName=hair_5_r}}
     {transformType=Rotation, targetBone {boneName=hair_6_r}}
     {transformType=Rotation, targetBone {boneName=hair_3_c}}
     {transformType=Rotation, targetBone {boneName=hair_1_c}}
     {transformType=Rotation, targetBone {boneName=hair_2_c}}
    ]
    constraintWeights [1,1,1,1,1,1]
   }
   n10 = "Constraint - hair_5_cst_r (RRRRRR)"!(
    ComponentPose=@n9
    Alpha=1
    ConstraintWeights_0=0
    ConstraintWeights_1:float=0.5
    ConstraintWeights_2:float=0.3
    ConstraintWeights_3:float=0.2
    ConstraintWeights_4:float=0
    ConstraintWeights_5:float=0
   ) -> (Pose) {
    alphaBoolBlend {blendOption=Linear}
    boneToModify {boneName=hair_5_cst_r}
    constraintSetup [
     {transformType=Rotation, targetBone {boneName=hair_4_r}}
     {transformType=Rotation, targetBone {boneName=hair_5_r}}
     {transformType=Rotation, targetBone {boneName=hair_6_r}}
     {transformType=Rotation, targetBone {boneName=hair_3_c}}
     {transformType=Rotation, targetBone {boneName=hair_1_c}}
     {transformType=Rotation, targetBone {boneName=hair_2_c}}
    ]
    constraintWeights [1,1,1,1,1,1]
   }
   n11 = "Constraint - hair_6_cst_r (RRRRRR)"!(
    ComponentPose=@n10
    Alpha=1
    ConstraintWeights_0=0.1
    ConstraintWeights_1:float=0.3
    ConstraintWeights_2:float=0.6
    ConstraintWeights_3:float=0
    ConstraintWeights_4:float=0
    ConstraintWeights_5:float=0
   ) -> (Pose) {
    alphaBoolBlend {blendOption=Linear}
    boneToModify {boneName=hair_6_cst_r}
    constraintSetup [
     {transformType=Rotation, targetBone {boneName=hair_4_r}}
     {transformType=Rotation, targetBone {boneName=hair_5_r}}
     {transformType=Rotation, targetBone {boneName=hair_6_r}}
     {transformType=Rotation, targetBone {boneName=hair_3_c}}
     {transformType=Rotation, targetBone {boneName=hair_1_c}}
     {transformType=Rotation, targetBone {boneName=hair_2_c}}
    ]
    constraintWeights [1,1,1,1,1,1]
   }
   n12 = "Constraint - hair_6_r_end (RRRRRR)"!(
    ComponentPose=@n11
    Alpha=1
    ConstraintWeights_0=0
    ConstraintWeights_1:float=0.9
    ConstraintWeights_2:float=0.1
    ConstraintWeights_3:float=0
    ConstraintWeights_4:float=0
    ConstraintWeights_5:float=0
   ) -> (Pose) {
    alphaBoolBlend {blendOption=Linear}
    boneToModify {boneName=hair_6_r_end}
    constraintSetup [
     {transformType=Rotation, targetBone {boneName=hair_4_r}}
     {transformType=Rotation, targetBone {boneName=hair_5_r}}
     {transformType=Rotation, targetBone {boneName=hair_6_r}}
     {transformType=Rotation, targetBone {boneName=hair_3_c}}
     {transformType=Rotation, targetBone {boneName=hair_1_c}}
     {transformType=Rotation, targetBone {boneName=hair_2_c}}
    ]
    constraintWeights [1,1,1,1,1,1]
   }
   n13 = "Constraint - hair_3_cst_c (RRRRRR)"!(
    ComponentPose=@n12
    Alpha=1
    ConstraintWeights_0=0
    ConstraintWeights_1:float=0
    ConstraintWeights_2:float=0.3
    ConstraintWeights_3:float=0.4
    ConstraintWeights_4:float=0.3
    ConstraintWeights_5:float=0
   ) -> (Pose) {
    alphaBoolBlend {blendOption=Linear}
    boneToModify {boneName=hair_3_cst_c}
    constraintSetup [
     {transformType=Rotation, targetBone {boneName=hair_4_r}}
     {transformType=Rotation, targetBone {boneName=hair_5_r}}
     {transformType=Rotation, targetBone {boneName=hair_6_r}}
     {transformType=Rotation, targetBone {boneName=hair_3_c}}
     {transformType=Rotation, targetBone {boneName=hair_1_c}}
     {transformType=Rotation, targetBone {boneName=hair_2_c}}
    ]
    constraintWeights [1,1,1,1,1,1]
   }
   n14 = "Constraint - hair_1_cst_c (RRRRRR)"!(
    ComponentPose=@n13
    Alpha=1
    ConstraintWeights_0=0
    ConstraintWeights_1:float=0
    ConstraintWeights_2:float=0
    ConstraintWeights_3:float=0.3
    ConstraintWeights_4:float=0.4
    ConstraintWeights_5:float=0.3
   ) -> (Pose) {
    alphaBoolBlend {blendOption=Linear}
    boneToModify {boneName=hair_1_cst_c}
    constraintSetup [
     {transformType=Rotation, targetBone {boneName=hair_4_r}}
     {transformType=Rotation, targetBone {boneName=hair_5_r}}
     {transformType=Rotation, targetBone {boneName=hair_6_r}}
     {transformType=Rotation, targetBone {boneName=hair_3_c}}
     {transformType=Rotation, targetBone {boneName=hair_1_c}}
     {transformType=Rotation, targetBone {boneName=hair_2_c}}
    ]
    constraintWeights [0,0,0,0.3,0.4,0.3]
   }
   n15 = "Constraint - hair_2_cst_c (RRRRRR)"!(
    ComponentPose=@n14
    Alpha=1
    ConstraintWeights_0=0
    ConstraintWeights_1:float=0
    ConstraintWeights_2:float=0
    ConstraintWeights_3:float=0
    ConstraintWeights_4:float=0.3
    ConstraintWeights_5:float=0.7
   ) -> (Pose) {
    alphaBoolBlend {blendOption=Linear}
    boneToModify {boneName=hair_2_cst_c}
    constraintSetup [
     {transformType=Rotation, targetBone {boneName=hair_4_r}}
     {transformType=Rotation, targetBone {boneName=hair_5_r}}
     {transformType=Rotation, targetBone {boneName=hair_6_r}}
     {transformType=Rotation, targetBone {boneName=hair_3_c}}
     {transformType=Rotation, targetBone {boneName=hair_1_c}}
     {transformType=Rotation, targetBone {boneName=hair_2_c}}
    ]
    constraintWeights [1,1,1,1,1,1]
   }
   n16 = Anim Dynamics - Bone: hair_6_l(ComponentPose=@n15, Alpha=1, ExternalForce="0,0,0") {
    angularDampingOverride=1
    angularSpringConstant=1500
    bAngularSpring=true
    bGravityOverrideInSimSpace=true
    bOverrideAngularDamping=true
    bUseSphericalLimits=true
    gravityScale=0.35
    simulationSpace=RootRelative
    alphaBoolBlend {blendOption=Linear}
    alphaScaleBiasClamp {scale=0.25}
    boundBone {boneName=hair_6_l}
    componentLinearAccScale {z=0.35}
    componentLinearVelScale {z=0.35}
    physicsBodyDefinitions [{
      collisionType=CustomSphere
      sphereCollisionRadius=2
      boundBone {boneName=hair_6_l}
      boxExtents=<10 20 20>
      constraintSetup {angularLimitsMax=<0 45 20>, angularLimitsMin=<0 -45 -80>, angularTarget=<1 0 0>}
      localJointOffset=<10 0 0>
     }]
    sphericalLimits [
     {limitRadius=10, drivingBone {boneName=neck_02}, sphereLocalOffset=<-10 0 0>}
     {limitRadius=10, drivingBone {boneName=clavicle_l}, sphereLocalOffset=<10 0 -3>}
     {limitRadius=10, drivingBone {boneName=clavicle_r}, sphereLocalOffset=<-10 0 3>}
     {limitRadius=8, drivingBone {boneName=head}, sphereLocalOffset=<0 0 0>}
    ]
   }
   n17 = Anim Dynamics - Bone: hair_7_l(ComponentPose=@n16, Alpha=1, ExternalForce="0,0,0") {
    angularDampingOverride=1
    angularSpringConstant=1500
    bAngularSpring=true
    bGravityOverrideInSimSpace=true
    bOverrideAngularDamping=true
    bUseSphericalLimits=true
    gravityScale=0.35
    simulationSpace=RootRelative
    alphaBoolBlend {blendOption=Linear}
    alphaScaleBiasClamp {scale=0.25}
    boundBone {boneName=hair_7_l}
    componentLinearAccScale {z=0.35}
    componentLinearVelScale {z=0.35}
    physicsBodyDefinitions [{
      collisionType=CustomSphere
      sphereCollisionRadius=2
      boundBone {boneName=hair_7_l}
      boxExtents=<10 20 20>
      constraintSetup {angularLimitsMax=<0 45 20>, angularLimitsMin=<0 -45 -80>, angularTarget=<1 0 0>}
      localJointOffset=<10 0 0>
     }]
    sphericalLimits [
     {limitRadius=10, drivingBone {boneName=neck_02}, sphereLocalOffset=<-10 0 0>}
     {limitRadius=10, drivingBone {boneName=clavicle_l}, sphereLocalOffset=<10 0 -3>}
     {limitRadius=10, drivingBone {boneName=clavicle_r}, sphereLocalOffset=<-10 0 3>}
     {limitRadius=8, drivingBone {boneName=head}, sphereLocalOffset=<0 0 0>}
    ]
   }
   n18 = Anim Dynamics - Bone: hair_1_c_end(ComponentPose=@n17, Alpha=1, ExternalForce="0,0,0") {
    angularDampingOverride=1
    angularSpringConstant=1500
    bAngularSpring=true
    bGravityOverrideInSimSpace=true
    bOverrideAngularDamping=true
    bUseSphericalLimits=true
    gravityScale=0.35
    simulationSpace=RootRelative
    alphaBoolBlend {blendOption=Linear}
    alphaScaleBiasClamp {scale=0.5}
    boundBone {boneName=hair_1_c_end}
    componentLinearAccScale {z=0.5}
    componentLinearVelScale {z=0.5}
    physicsBodyDefinitions [{
      collisionType=CustomSphere
      sphereCollisionRadius=2
      boundBone {boneName=hair_1_c_end}
      boxExtents=<10 20 20>
      constraintSetup {angularLimitsMax=<0 80 20>, angularLimitsMin=<0 -80 -80>, angularTarget=<0 0 0>}
      localJointOffset=<10 0 0>
     }]
    sphericalLimits [
     {limitRadius=10, drivingBone {boneName=neck_02}, sphereLocalOffset=<-10 0 0>}
     {limitRadius=10, drivingBone {boneName=clavicle_l}, sphereLocalOffset=<10 0 -3>}
     {limitRadius=10, drivingBone {boneName=clavicle_r}, sphereLocalOffset=<-10 0 3>}
    ]
   }
   n19 = Anim Dynamics - Bone: hair_1_l_end(ComponentPose=@n18, Alpha=1, ExternalForce="0,0,0") {
    angularDampingOverride=1
    angularSpringConstant=1500
    bAngularSpring=true
    bGravityOverrideInSimSpace=true
    bOverrideAngularDamping=true
    bUseSphericalLimits=true
    gravityScale=0.35
    simulationSpace=RootRelative
    alphaBoolBlend {blendOption=Linear}
    alphaScaleBiasClamp {scale=0.5}
    boundBone {boneName=hair_1_l_end}
    componentLinearAccScale {z=0.5}
    componentLinearVelScale {z=0.5}
    physicsBodyDefinitions [{
      collisionType=CustomSphere
      sphereCollisionRadius=2
      boundBone {boneName=hair_1_l_end}
      boxExtents=<10 20 20>
      constraintSetup {angularLimitsMax=<0 80 20>, angularLimitsMin=<0 -80 -80>, angularTarget=<0 0 0>}
      localJointOffset=<10 0 0>
     }]
    sphericalLimits [
     {limitRadius=10, drivingBone {boneName=neck_02}, sphereLocalOffset=<-10 0 0>}
     {limitRadius=10, drivingBone {boneName=clavicle_l}, sphereLocalOffset=<10 0 -3>}
     {limitRadius=10, drivingBone {boneName=clavicle_r}, sphereLocalOffset=<-10 0 3>}
    ]
   }
   n20 = Anim Dynamics - Bone: hair_1_r_end(ComponentPose=@n19, Alpha=1, ExternalForce="0,0,0") {
    angularDampingOverride=1
    angularSpringConstant=1500
    bAngularSpring=true
    bGravityOverrideInSimSpace=true
    bOverrideAngularDamping=true
    bUseSphericalLimits=true
    gravityScale=0.35
    simulationSpace=RootRelative
    alphaBoolBlend {blendOption=Linear}
    alphaScaleBiasClamp {scale=0.5}
    boundBone {boneName=hair_1_r_end}
    componentLinearAccScale {z=0.5}
    componentLinearVelScale {z=0.5}
    physicsBodyDefinitions [{
      collisionType=CustomSphere
      sphereCollisionRadius=2
      boundBone {boneName=hair_1_r_end}
      boxExtents=<10 20 20>
      constraintSetup {angularLimitsMax=<0 80 20>, angularLimitsMin=<0 -80 -80>, angularTarget=<0 0 0>}
      localJointOffset=<10 0 0>
     }]
    sphericalLimits [
     {limitRadius=10, drivingBone {boneName=neck_02}, sphereLocalOffset=<-10 0 0>}
     {limitRadius=10, drivingBone {boneName=clavicle_l}, sphereLocalOffset=<10 0 -3>}
     {limitRadius=10, drivingBone {boneName=clavicle_r}, sphereLocalOffset=<-10 0 3>}
    ]
   }
   n21 = Anim Dynamics - Bone: hair_2_c_end(ComponentPose=@n20, Alpha=1, ExternalForce="0,0,0") {
    angularDampingOverride=1
    angularSpringConstant=1500
    bAngularSpring=true
    bGravityOverrideInSimSpace=true
    bOverrideAngularDamping=true
    bUseSphericalLimits=true
    gravityScale=0.35
    simulationSpace=RootRelative
    alphaBoolBlend {blendOption=Linear}
    alphaScaleBiasClamp {scale=0.5}
    boundBone {boneName=hair_2_c_end}
    componentLinearAccScale {z=0.5}
    componentLinearVelScale {z=0.5}
    physicsBodyDefinitions [{
      collisionType=CustomSphere
      sphereCollisionRadius=2
      boundBone {boneName=hair_2_c_end}
      boxExtents=<10 20 20>
      constraintSetup {angularLimitsMax=<0 80 20>, angularLimitsMin=<0 -80 -80>, angularTarget=<0 0 0>}
      localJointOffset=<10 0 0>
     }]
    sphericalLimits [
     {limitRadius=10, drivingBone {boneName=neck_02}, sphereLocalOffset=<-10 0 0>}
     {limitRadius=10, drivingBone {boneName=clavicle_l}, sphereLocalOffset=<10 0 -3>}
     {limitRadius=10, drivingBone {boneName=clavicle_r}, sphereLocalOffset=<-10 0 3>}
    ]
   }
   n22 = Anim Dynamics - Bone: hair_2_l_end(ComponentPose=@n21, Alpha=1, ExternalForce="0,0,0") {
    angularDampingOverride=1
    angularSpringConstant=1500
    bAngularSpring=true
    bGravityOverrideInSimSpace=true
    bOverrideAngularDamping=true
    bUseSphericalLimits=true
    gravityScale=0.35
    simulationSpace=RootRelative
    alphaBoolBlend {blendOption=Linear}
    alphaScaleBiasClamp {scale=0.5}
    boundBone {boneName=hair_2_l_end}
    componentLinearAccScale {z=0.5}
    componentLinearVelScale {z=0.5}
    physicsBodyDefinitions [{
      collisionType=CustomSphere
      sphereCollisionRadius=2
      boundBone {boneName=hair_2_l_end}
      boxExtents=<10 20 20>
      constraintSetup {angularLimitsMax=<0 0 20>, angularLimitsMin=<0 0 -80>, angularTarget=<0 0 0>}
      localJointOffset=<10 0 0>
     }]
    sphericalLimits [
     {limitRadius=10, drivingBone {boneName=neck_02}, sphereLocalOffset=<-10 0 0>}
     {limitRadius=12, drivingBone {boneName=clavicle_l}, sphereLocalOffset=<10 0 -3>}
     {limitRadius=10, drivingBone {boneName=clavicle_r}, sphereLocalOffset=<-10 0 3>}
    ]
   }
   n23 = Anim Dynamics - Bone: hair_2_r_end(ComponentPose=@n22, Alpha=1, ExternalForce="0,0,0") {
    angularDampingOverride=1
    angularSpringConstant=1500
    bAngularSpring=true
    bGravityOverrideInSimSpace=true
    bOverrideAngularDamping=true
    bUseSphericalLimits=true
    gravityScale=0.35
    simulationSpace=RootRelative
    alphaBoolBlend {blendOption=Linear}
    alphaScaleBiasClamp {scale=0.5}
    boundBone {boneName=hair_2_r_end}
    componentLinearAccScale {z=0.5}
    componentLinearVelScale {z=0.5}
    physicsBodyDefinitions [{
      collisionType=CustomSphere
      sphereCollisionRadius=2
      boundBone {boneName=hair_2_r_end}
      boxExtents=<10 20 20>
      constraintSetup {angularLimitsMax=<0 20 20>, angularLimitsMin=<0 -80 -80>, angularTarget=<0 0 0>}
      localJointOffset=<2 0 0>
     }]
    sphericalLimits [
     {limitRadius=10, drivingBone {boneName=neck_02}, sphereLocalOffset=<-10 0 0>}
     {limitRadius=10, drivingBone {boneName=clavicle_l}, sphereLocalOffset=<10 0 -3>}
     {limitRadius=10, drivingBone {boneName=clavicle_r}, sphereLocalOffset=<-10 0 3>}
    ]
   }
   n24 = Anim Dynamics - Bone: hair_3_c_end(ComponentPose=@n23, Alpha=1, ExternalForce="0,0,0") {
    angularDampingOverride=1
    angularSpringConstant=1500
    bAngularSpring=true
    bGravityOverrideInSimSpace=true
    bOverrideAngularDamping=true
    bUseSphericalLimits=true
    gravityScale=0.35
    simulationSpace=RootRelative
    alphaBoolBlend {blendOption=Linear}
    alphaScaleBiasClamp {scale=0.5}
    boundBone {boneName=hair_3_c_end}
    componentLinearAccScale {z=0.5}
    componentLinearVelScale {z=0.5}
    physicsBodyDefinitions [{
      collisionType=CustomSphere
      sphereCollisionRadius=2
      boundBone {boneName=hair_3_c_end}
      boxExtents=<10 20 20>
      constraintSetup {angularLimitsMax=<0 20 20>, angularLimitsMin=<0 -80 -80>, angularTarget=<0 0 0>}
      localJointOffset=<10 0 0>
     }]
    sphericalLimits [
     {limitRadius=10, drivingBone {boneName=neck_02}, sphereLocalOffset=<-10 0 0>}
     {limitRadius=10, drivingBone {boneName=clavicle_l}, sphereLocalOffset=<10 0 -3>}
     {limitRadius=10, drivingBone {boneName=clavicle_r}, sphereLocalOffset=<-10 0 3>}
    ]
   }
   n25 = Anim Dynamics - Bone: hair_3_l_end(ComponentPose=@n24, Alpha=1, ExternalForce="0,0,0") {
    angularDampingOverride=1
    angularSpringConstant=1500
    bAngularSpring=true
    bGravityOverrideInSimSpace=true
    bOverrideAngularDamping=true
    bUseSphericalLimits=true
    gravityScale=0.35
    simulationSpace=RootRelative
    alphaBoolBlend {blendOption=Linear}
    alphaScaleBiasClamp {scale=0.5}
    boundBone {boneName=hair_3_l_end}
    componentLinearAccScale {z=0.5}
    componentLinearVelScale {z=0.5}
    physicsBodyDefinitions [{
      collisionType=CustomSphere
      sphereCollisionRadius=2
      boundBone {boneName=hair_3_l_end}
      boxExtents=<10 20 20>
      constraintSetup {angularLimitsMax=<0 0 20>, angularLimitsMin=<0 0 -80>, angularTarget=<0 0 0>}
      localJointOffset=<10 0 0>
     }]
    sphericalLimits [
     {limitRadius=10, drivingBone {boneName=neck_02}, sphereLocalOffset=<-10 0 0>}
     {limitRadius=13, drivingBone {boneName=clavicle_l}, sphereLocalOffset=<10 0 -3>}
     {limitRadius=10, drivingBone {boneName=clavicle_r}, sphereLocalOffset=<-10 0 3>}
    ]
   }
   n26 = Anim Dynamics - Bone: hair_3_r_end(ComponentPose=@n25, Alpha=1, ExternalForce="0,0,0") {
    angularDampingOverride=1
    angularSpringConstant=1500
    bAngularSpring=true
    bGravityOverrideInSimSpace=true
    bOverrideAngularDamping=true
    bUseSphericalLimits=true
    gravityScale=0.35
    simulationSpace=RootRelative
    alphaBoolBlend {blendOption=Linear}
    alphaScaleBiasClamp {scale=0.5}
    boundBone {boneName=hair_3_r_end}
    componentLinearAccScale {z=0.5}
    componentLinearVelScale {z=0.5}
    physicsBodyDefinitions [{
      collisionType=CustomSphere
      sphereCollisionRadius=2
      boundBone {boneName=hair_3_r_end}
      boxExtents=<10 20 20>
      constraintSetup {angularLimitsMax=<0 20 20>, angularLimitsMin=<0 -80 -80>, angularTarget=<0 0 0>}
      localJointOffset=<10 0 0>
     }]
    sphericalLimits [
     {limitRadius=10, drivingBone {boneName=neck_02}, sphereLocalOffset=<-10 0 0>}
     {limitRadius=10, drivingBone {boneName=clavicle_l}, sphereLocalOffset=<10 0 -3>}
     {limitRadius=10, drivingBone {boneName=clavicle_r}, sphereLocalOffset=<-10 0 3>}
    ]
   }
   n27 = Anim Dynamics - Bone: hair_4_l_end(ComponentPose=@n26, Alpha=1, ExternalForce="0,0,0") {
    angularDampingOverride=1
    angularSpringConstant=1500
    bAngularSpring=true
    bGravityOverrideInSimSpace=true
    bOverrideAngularDamping=true
    bUseSphericalLimits=true
    gravityScale=0.35
    simulationSpace=RootRelative
    alphaBoolBlend {blendOption=Linear}
    alphaScaleBiasClamp {scale=0.5}
    boundBone {boneName=hair_4_l_end}
    componentLinearAccScale {z=0.5}
    componentLinearVelScale {z=0.5}
    physicsBodyDefinitions [{
      collisionType=CustomSphere
      sphereCollisionRadius=2
      boundBone {boneName=hair_4_l_end}
      boxExtents=<10 20 20>
      constraintSetup {angularLimitsMax=<0 20 20>, angularLimitsMin=<0 -20 -80>, angularTarget=<0 0 0>}
      localJointOffset=<5 0 0>
     }]
    sphericalLimits [
     {limitRadius=15, drivingBone {boneName=neck_02}, sphereLocalOffset=<-10 0 0>}
     {limitRadius=12, drivingBone {boneName=clavicle_l}, sphereLocalOffset=<10 0 -3>}
     {limitRadius=10, drivingBone {boneName=clavicle_r}, sphereLocalOffset=<-10 0 3>}
    ]
   }
   n28 = Anim Dynamics - Bone: hair_4_r_end(ComponentPose=@n27, Alpha=1, ExternalForce="0,0,0") {
    angularDampingOverride=1
    angularSpringConstant=1500
    bAngularSpring=true
    bGravityOverrideInSimSpace=true
    bOverrideAngularDamping=true
    bUseSphericalLimits=true
    gravityScale=0.35
    simulationSpace=RootRelative
    alphaBoolBlend {blendOption=Linear}
    alphaScaleBiasClamp {scale=0.5}
    boundBone {boneName=hair_4_r_end}
    componentLinearAccScale {z=0.5}
    componentLinearVelScale {z=0.5}
    physicsBodyDefinitions [{
      collisionType=CustomSphere
      sphereCollisionRadius=2
      boundBone {boneName=hair_4_r_end}
      boxExtents=<10 20 20>
      constraintSetup {angularLimitsMax=<0 45 20>, angularLimitsMin=<0 -45 -80>, angularTarget=<0 0 0>}
      localJointOffset=<10 0 0>
     }]
    sphericalLimits [
     {limitRadius=10, drivingBone {boneName=neck_02}, sphereLocalOffset=<-10 0 0>}
     {limitRadius=10, drivingBone {boneName=clavicle_l}, sphereLocalOffset=<10 0 -3>}
     {limitRadius=10, drivingBone {boneName=clavicle_r}, sphereLocalOffset=<-10 0 3>}
     {limitRadius=0, drivingBone {boneName=None}, sphereLocalOffset=<0 0 0>}
    ]
   }
   n29 = Anim Dynamics - Bone: hair_5_l_end(ComponentPose=@n28, Alpha=1, ExternalForce="0,0,0") {
    angularDampingOverride=1
    angularSpringConstant=1500
    bAngularSpring=true
    bGravityOverrideInSimSpace=true
    bOverrideAngularDamping=true
    bUseSphericalLimits=true
    gravityScale=0.35
    simulationSpace=RootRelative
    alphaBoolBlend {blendOption=Linear}
    alphaScaleBiasClamp {scale=0.5}
    boundBone {boneName=hair_5_l_end}
    componentLinearAccScale {z=0.5}
    componentLinearVelScale {z=0.5}
    physicsBodyDefinitions [{
      collisionType=CustomSphere
      sphereCollisionRadius=2
      boundBone {boneName=hair_5_l_end}
      boxExtents=<10 20 20>
      constraintSetup {angularLimitsMax=<0 0 20>, angularLimitsMin=<0 -10 -80>, angularTarget=<0 0 0>}
      localJointOffset=<5 0 0>
     }]
    sphericalLimits [
     {limitRadius=10, drivingBone {boneName=neck_02}, sphereLocalOffset=<-10 0 0>}
     {limitRadius=10, drivingBone {boneName=clavicle_l}, sphereLocalOffset=<10 0 -3>}
     {limitRadius=10, drivingBone {boneName=clavicle_r}, sphereLocalOffset=<-10 0 3>}
    ]
   }
   n30 = Anim Dynamics - Bone: hair_5_r_end(ComponentPose=@n29, Alpha=1, ExternalForce="0,0,0") {
    angularDampingOverride=1
    angularSpringConstant=1500
    bAngularSpring=true
    bGravityOverrideInSimSpace=true
    bOverrideAngularDamping=true
    bUsePlanarLimit=false
    bUseSphericalLimits=true
    gravityScale=0.35
    simulationSpace=RootRelative
    alphaBoolBlend {blendOption=Linear}
    alphaScaleBiasClamp {scale=0.5}
    boundBone {boneName=hair_5_r_end}
    componentLinearAccScale {z=0.5}
    componentLinearVelScale {z=0.5}
    physicsBodyDefinitions [{
      collisionType=CustomSphere
      sphereCollisionRadius=2
      boundBone {boneName=hair_5_r_end}
      boxExtents=<10 20 20>
      constraintSetup {angularLimitsMax=<0 80 20>, angularLimitsMin=<0 -80 -80>, angularTarget=<0 0 0>}
      localJointOffset=<7 0 0>
     }]
    planarLimits [
     {drivingBone {boneName=clavicle_r}, planeTransform {location=<0 0 0>, rotation=<rot 0 0 0>, scale=<0.01 0 0>}}
     {drivingBone {boneName=None}, planeTransform {location=<0 0 0>, rotation=<rot 0 0 0>}}
    ]
    sphericalLimits [
     {limitRadius=10, drivingBone {boneName=neck_02}, sphereLocalOffset=<-10 0 0>}
     {limitRadius=10, drivingBone {boneName=clavicle_l}, sphereLocalOffset=<10 0 -3>}
     {limitRadius=10, drivingBone {boneName=clavicle_r}, sphereLocalOffset=<-10 0 3>}
     {limitRadius=5, drivingBone {boneName=None}, sphereLocalOffset=<-10 0 3>}
    ]
   }
   n31 = Anim Dynamics - Bone: hair_6_l_end(ComponentPose=@n30, Alpha=1, ExternalForce="0,0,0") {
    angularDampingOverride=1
    angularSpringConstant=1500
    bAngularSpring=true
    bGravityOverrideInSimSpace=true
    bOverrideAngularDamping=true
    bUseSphericalLimits=true
    gravityScale=0.35
    simulationSpace=RootRelative
    alphaBoolBlend {blendOption=Linear}
    alphaScaleBiasClamp {scale=0.5}
    boundBone {boneName=hair_6_l_end}
    componentLinearAccScale {z=0.5}
    componentLinearVelScale {z=0.5}
    physicsBodyDefinitions [{
      collisionType=CustomSphere
      sphereCollisionRadius=2
      boundBone {boneName=hair_6_l_end}
      boxExtents=<10 20 20>
      constraintSetup {angularLimitsMax=<0 80 20>, angularLimitsMin=<0 -80 -80>, angularTarget=<0 0 0>}
      localJointOffset=<3 0 0>
     }]
    sphericalLimits [
     {limitRadius=10, drivingBone {boneName=neck_02}, sphereLocalOffset=<-10 0 0>}
     {limitRadius=10, drivingBone {boneName=clavicle_l}, sphereLocalOffset=<10 0 -3>}
     {limitRadius=10, drivingBone {boneName=clavicle_r}, sphereLocalOffset=<-10 0 3>}
     {limitRadius=0, drivingBone {boneName=head}, sphereLocalOffset=<0 0 0>}
    ]
   }
   n32 = Anim Dynamics - Bone: hair_6_r_end(ComponentPose=@n31, Alpha=1, ExternalForce="0,0,0") {
    angularDampingOverride=1
    angularSpringConstant=1500
    bAngularSpring=true
    bGravityOverrideInSimSpace=true
    bOverrideAngularDamping=true
    bUseSphericalLimits=true
    gravityScale=0.35
    simulationSpace=RootRelative
    alphaBoolBlend {blendOption=Linear}
    alphaScaleBiasClamp {scale=0.5}
    boundBone {boneName=hair_6_r_end}
    componentLinearAccScale {z=0.5}
    componentLinearVelScale {z=0.5}
    physicsBodyDefinitions [{
      collisionType=CustomSphere
      sphereCollisionRadius=4
      boundBone {boneName=hair_6_r_end}
      boxExtents=<10 20 20>
      constraintSetup {angularLimitsMax=<0 45 20>, angularLimitsMin=<0 -45 -80>, angularTarget=<0 0 0>}
      localJointOffset=<10 0 0>
     }]
    sphericalLimits [
     {limitRadius=12, drivingBone {boneName=neck_02}, sphereLocalOffset=<-10 0 0>}
     {limitRadius=10, drivingBone {boneName=clavicle_l}, sphereLocalOffset=<10 0 -3>}
     {limitRadius=10, drivingBone {boneName=clavicle_r}, sphereLocalOffset=<-10 0 3>}
     {limitRadius=10, drivingBone {boneName=head}, sphereLocalOffset=<0 0 0>}
    ]
   }
   n33 = Anim Dynamics - Bone: hair_7_l_end(ComponentPose=@n32, Alpha=1, ExternalForce="0,0,0") {
    angularDampingOverride=1
    angularSpringConstant=1500
    bAngularSpring=true
    bGravityOverrideInSimSpace=true
    bOverrideAngularDamping=true
    bUseSphericalLimits=true
    gravityScale=0.35
    simulationSpace=RootRelative
    alphaBoolBlend {blendOption=Linear}
    alphaScaleBiasClamp {scale=0.5}
    boundBone {boneName=hair_7_l_end}
    componentLinearAccScale {z=0.5}
    componentLinearVelScale {z=0.5}
    physicsBodyDefinitions [{
      collisionType=CustomSphere
      sphereCollisionRadius=2
      boundBone {boneName=hair_7_l_end}
      boxExtents=<10 20 20>
      constraintSetup {angularLimitsMax=<0 80 20>, angularLimitsMin=<0 -80 -80>, angularTarget=<0 0 0>}
      localJointOffset=<4 0 0>
     }]
    sphericalLimits [
     {limitRadius=10, drivingBone {boneName=neck_02}, sphereLocalOffset=<-10 0 0>}
     {limitRadius=10, drivingBone {boneName=clavicle_l}, sphereLocalOffset=<10 0 -3>}
     {limitRadius=10, drivingBone {boneName=clavicle_r}, sphereLocalOffset=<-10 0 3>}
     {limitRadius=10, drivingBone {boneName=head}, sphereLocalOffset=<0 0 0>}
     {limitRadius=0, drivingBone {boneName=None}, sphereLocalOffset=<0 0 0>}
    ]
   }
   n34 = Control Rig(Source=Component To Local(ComponentPose=@n33)) {
    alphaBoolBlend {blendOption=Linear}
    controlRigAssetReference {blueprintRigClass=&/F_MED_Pink_Trooper_FaceAcc_CtrlRig.F_MED_Pink_Trooper_FaceAcc_CtrlRig_C}
   }
   Output Pose(Result=@n34) {layerGroup=None}
   n35 = "Constraint - hair_group_r_tgt (PPP)"!(
    ComponentPose=Local To Component(LocalPose=Use cached pose 'orig_cst'())
    Alpha=1
    ConstraintWeights_0=0.2146
    ConstraintWeights_1:float=0
    ConstraintWeights_2:float=0.7854
   ) -> (Pose) {
    alphaBoolBlend {blendOption=Linear}
    boneToModify {boneName=hair_group_r_tgt}
    constraintSetup [{targetBone {boneName=clavicle_r}}, {targetBone {boneName=spine_04}}, {targetBone {boneName=spine_05}}]
    constraintWeights [1,1,1]
   }
   n36 = "Constraint - hair_group_l_tgt (PP)"!(ComponentPose=@n35, Alpha=1, ConstraintWeights_0=0.7855, ConstraintWeights_1:float=0.2145) -> (Pose) {
    alphaBoolBlend {blendOption=Linear}
    boneToModify {boneName=hair_group_l_tgt}
    constraintSetup [{targetBone {boneName=spine_05}}, {targetBone {boneName=clavicle_l}}]
    constraintWeights [1,1]
   }
   n37 = "Constraint - hair_1_c_tgt (P)"!(ComponentPose=@n36, Alpha=1, ConstraintWeights_0=1, ConstraintWeights_1:float=0) -> (Pose) {
    alphaBoolBlend {blendOption=Linear}
    boneToModify {boneName=hair_1_c_tgt}
    constraintSetup [{targetBone {boneName=spine_05}}]
    constraintWeights [1,0]
   }
   n38 = "Constraint - hair_2_c_tgt (PP)"!(ComponentPose=@n37, Alpha=1, ConstraintWeights_0=0.8771, ConstraintWeights_1:float=0.1229) -> (Pose) {
    alphaBoolBlend {blendOption=Linear}
    boneToModify {boneName=hair_2_c_tgt}
    constraintSetup [{targetBone {boneName=spine_05}}, {targetBone {boneName=clavicle_l}}]
    constraintWeights [1,1]
   }
   n39 = "Constraint - hair_3_c_tgt (PPP)"!(ComponentPose=@n38, Alpha=1, ConstraintWeights_0=0.012, ConstraintWeights_1:float=0, ConstraintWeights_2:float=0.988) -> (Pose) {
    alphaBoolBlend {blendOption=Linear}
    boneToModify {boneName=hair_3_c_tgt}
    constraintSetup [{targetBone {boneName=clavicle_r}}, {targetBone {boneName=spine_04}}, {targetBone {boneName=spine_05}}]
    constraintWeights [1,1,1]
   }
   n40 = "Constraint - hair_5_l_tgt (PPP)"!(ComponentPose=@n39, Alpha=1, ConstraintWeights_0=0.0122, ConstraintWeights_1:float=0.4572, ConstraintWeights_2:float=0.5306) -> (Pose) {
    alphaBoolBlend {blendOption=Linear}
    boneToModify {boneName=hair_5_l_tgt}
    constraintSetup [{targetBone {boneName=spine_04}}, {targetBone {boneName=spine_05}}, {targetBone {boneName=clavicle_l}}]
    constraintWeights [1,1,1]
   }
   n41 = "Constraint - hair_5_r_tgt (PPP)"!(ComponentPose=@n40, Alpha=1, ConstraintWeights_0=0.3181, ConstraintWeights_1:float=0, ConstraintWeights_2:float=0.6819) -> (Pose) {
    alphaBoolBlend {blendOption=Linear}
    boneToModify {boneName=hair_5_r_tgt}
    constraintSetup [{targetBone {boneName=clavicle_r}}, {targetBone {boneName=spine_04}}, {targetBone {boneName=spine_05}}]
    constraintWeights [1,1,1]
   }
   n42 = "Transform (Modify) Bone - Bone: hair_group_r_tgt"(ComponentPose=@n41, Alpha=1, Translation="10,0,0", Rotation="0,0,0", Scale="1,1,1") {
    translationMode=BMM_Additive
    translationSpace=BCS_BoneSpace
    alphaBoolBlend {blendOption=Linear}
    boneToModify {boneName=hair_group_r_tgt}
    translation {x=10}
   }
   n43 = "Transform (Modify) Bone - Bone: hair_group_l_tgt"(ComponentPose=@n42, Alpha=1, Translation="10,0,0", Rotation="0,0,0", Scale="1,1,1") {
    translationMode=BMM_Additive
    translationSpace=BCS_BoneSpace
    alphaBoolBlend {blendOption=Linear}
    boneToModify {boneName=hair_group_l_tgt}
    translation {x=10}
   }
   n44 = "Transform (Modify) Bone - Bone: hair_1_c_tgt"(ComponentPose=@n43, Alpha=1, Translation="10,0,-10", Rotation="0,0,0", Scale="1,1,1") {
    translationMode=BMM_Additive
    translationSpace=BCS_BoneSpace
    alphaBoolBlend {blendOption=Linear}
    boneToModify {boneName=hair_1_c_tgt}
    translation {x=10, z=-10}
   }
   n45 = "Transform (Modify) Bone - Bone: hair_2_c_tgt"(ComponentPose=@n44, Alpha=1, Translation="10,0,-5", Rotation="0,0,0", Scale="1,1,1") {
    translationMode=BMM_Additive
    translationSpace=BCS_BoneSpace
    alphaBoolBlend {blendOption=Linear}
    boneToModify {boneName=hair_2_c_tgt}
    translation {x=10, z=-5}
   }
   n46 = "Transform (Modify) Bone - Bone: hair_3_c_tgt"(ComponentPose=@n45, Alpha=1, Translation="10,0,-5", Rotation="0,0,0", Scale="1,1,1") {
    translationMode=BMM_Additive
    translationSpace=BCS_BoneSpace
    alphaBoolBlend {blendOption=Linear}
    boneToModify {boneName=hair_3_c_tgt}
    translation {x=10, z=-5}
   }
   n47 = "Transform (Modify) Bone - Bone: hair_5_l_tgt"(ComponentPose=@n46, Alpha=1, Translation="10,0,-5", Rotation="0,0,0", Scale="1,1,1") {
    translationMode=BMM_Additive
    translationSpace=BCS_BoneSpace
    alphaBoolBlend {blendOption=Linear}
    boneToModify {boneName=hair_5_l_tgt}
    translation {x=10, z=-5}
   }
   n48 = "Transform (Modify) Bone - Bone: hair_5_r_tgt"(ComponentPose=@n47, Alpha=1, Translation="10,0,-20", Rotation="0,0,0", Scale="1,1,1") {
    translationMode=BMM_Additive
    translationSpace=BCS_BoneSpace
    alphaBoolBlend {blendOption=Linear}
    boneToModify {boneName=hair_5_r_tgt}
    translation {x=10, z=-20}
   }
   n49 = "Transform (Modify) Bone - Bone: hair_5_r_tgt"!(ComponentPose=@n48, bAlphaBoolEnabled:bool=Get bIsTargeting(), Translation="10,-20,-40", Rotation="0,0,0", Scale="1,1,1") -> (Pose) {
    alphaInputType=Bool
    translationMode=BMM_Additive
    translationSpace=BCS_BoneSpace
    alphaBoolBlend {blendOption=Linear, blendOutTime=1}
    boneToModify {boneName=hair_5_r_tgt}
    translation=<10 -20 -40>
   }
   n50 = Get skydive_alpha() -> (skydive_alpha:double)
   n51 = "Transform (Modify) Bone - Bone: hair_group_r_tgt"!(ComponentPose=@n49, Alpha=@n50, Rotation="0,0,0", Scale="1,1,1") -> (Pose) {
    rotationSpace=BCS_ParentBoneSpace
    translationMode=BMM_Additive
    translationSpace=BCS_ParentBoneSpace
    alphaBoolBlend {blendOption=Linear}
    alphaScaleBiasClamp {bInterpResult=true}
    boneToModify {boneName=hair_group_r_tgt}
    translation=<-20 -15>
   }
   n52 = "Transform (Modify) Bone - Bone: hair_group_l_tgt"!(ComponentPose=@n51, Alpha=@n50, Rotation="0,0,0", Scale="1,1,1") -> (Pose) {
    rotationSpace=BCS_ParentBoneSpace
    translationMode=BMM_Additive
    translationSpace=BCS_ParentBoneSpace
    alphaBoolBlend {blendOption=Linear}
    alphaScaleBiasClamp {bInterpResult=true, scale=0.4}
    boneToModify {boneName=hair_group_l_tgt}
    translation=<-20 -15>
   }
   n53 = "Transform (Modify) Bone - Bone: hair_1_c_tgt"!(ComponentPose=@n52, Alpha=@n50, Rotation="0,0,0", Scale="1,1,1") -> (Pose) {
    rotationSpace=BCS_ParentBoneSpace
    translationMode=BMM_Additive
    translationSpace=BCS_ParentBoneSpace
    alphaBoolBlend {blendOption=Linear}
    alphaScaleBiasClamp {bInterpResult=true, scale=0.4}
    boneToModify {boneName=hair_1_c_tgt}
    translation=<-40 -15>
   }
   n54 = "Transform (Modify) Bone - Bone: hair_2_c_tgt"!(ComponentPose=@n53, Alpha=@n50, Rotation="0,0,0", Scale="1,1,1") -> (Pose) {
    rotationSpace=BCS_ParentBoneSpace
    translationMode=BMM_Additive
    translationSpace=BCS_ParentBoneSpace
    alphaBoolBlend {blendOption=Linear}
    alphaScaleBiasClamp {bInterpResult=true, scale=0.4}
    boneToModify {boneName=hair_2_c_tgt}
    translation=<-40 -15>
   }
   n55 = "Transform (Modify) Bone - Bone: hair_3_c_tgt"!(ComponentPose=@n54, Alpha=@n50, Rotation="0,0,0", Scale="1,1,1") -> (Pose) {
    rotationSpace=BCS_ParentBoneSpace
    translationMode=BMM_Additive
    translationSpace=BCS_ParentBoneSpace
    alphaBoolBlend {blendOption=Linear}
    alphaScaleBiasClamp {bInterpResult=true}
    boneToModify {boneName=hair_3_c_tgt}
    translation=<-30 -15>
   }
   n56 = "Transform (Modify) Bone - Bone: hair_5_l_tgt"!(ComponentPose=@n55, Alpha=@n50, Rotation="0,0,0", Scale="1,1,1") -> (Pose) {
    rotationSpace=BCS_ParentBoneSpace
    translationMode=BMM_Additive
    translationSpace=BCS_ParentBoneSpace
    alphaBoolBlend {blendOption=Linear}
    alphaScaleBiasClamp {bInterpResult=true, scale=0.4}
    boneToModify {boneName=hair_5_l_tgt}
    translation=<-20 -15>
   }
   n57 = "Transform (Modify) Bone - Bone: hair_5_r_tgt"!(ComponentPose=@n56, Alpha:float=@n50, Rotation="0,0,0", Scale="1,1,1") -> (Pose) {
    rotationSpace=BCS_ParentBoneSpace
    translationMode=BMM_Additive
    translationSpace=BCS_ParentBoneSpace
    alphaBoolBlend {blendOption=Linear}
    alphaScaleBiasClamp {bInterpResult=true}
    boneToModify {boneName=hair_5_r_tgt}
   }
   n58 = Look At - Bone: hair_group_r(ComponentPose=@n57, Alpha=0.6) {
    alphaBoolBlend {blendOption=Linear}
    boneToModify {boneName=hair_group_r}
    lookAtLocation {x=0}
    lookAtTarget {boneReference {boneName=hair_group_r_tgt}}
    lookAt_Axis {axis=<1 0>}
   }
   n59 = Look At - Bone: hair_group_l(ComponentPose=@n58, Alpha=0.6) {
    alphaBoolBlend {blendOption=Linear}
    boneToModify {boneName=hair_group_l}
    lookAtLocation {x=0}
    lookAtTarget {boneReference {boneName=hair_group_l_tgt}}
    lookAt_Axis {axis=<1 0>}
   }
   n60 = Look At - Bone: hair_1_c(ComponentPose=@n59, Alpha=1) {
    alphaBoolBlend {blendOption=Linear}
    boneToModify {boneName=hair_1_c}
    lookAtLocation {x=0}
    lookAtTarget {boneReference {boneName=hair_1_c_tgt}}
    lookAt_Axis {axis=<1 0>}
   }
   n61 = Look At - Bone: hair_2_c(ComponentPose=@n60, Alpha=1) {
    alphaBoolBlend {blendOption=Linear}
    boneToModify {boneName=hair_2_c}
    lookAtLocation {x=0}
    lookAtTarget {boneReference {boneName=hair_2_c_tgt}}
    lookAt_Axis {axis=<1 0>}
   }
   n62 = Look At - Bone: hair_3_c(ComponentPose=@n61, Alpha=1) {
    alphaBoolBlend {blendOption=Linear}
    boneToModify {boneName=hair_3_c}
    lookAtLocation {x=0}
    lookAtTarget {boneReference {boneName=hair_3_c_tgt}}
    lookAt_Axis {axis=<1 0>}
   }
   n63 = Look At - Bone: hair_5_l(ComponentPose=@n62, Alpha=1) {
    alphaBoolBlend {blendOption=Linear}
    boneToModify {boneName=hair_5_l}
    lookAtLocation {x=0}
    lookAtTarget {boneReference {boneName=hair_5_l_tgt}}
    lookAt_Axis {axis=<1 0>}
   }
   n64 = Look At - Bone: hair_5_r(ComponentPose=@n63, Alpha=1) {
    alphaBoolBlend {blendOption=Linear}
    boneToModify {boneName=hair_5_r}
    lookAtLocation {x=0}
    lookAtTarget {boneReference {boneName=hair_5_r_tgt}}
    lookAt_Axis {axis=<1 0>}
   }
   n65 = Look At - Bone: hair_6_r(ComponentPose=@n64, Alpha=1) {
    alphaBoolBlend {blendOption=Linear}
    boneToModify {boneName=hair_6_r}
    lookAtLocation {x=0}
    lookAtTarget {boneReference {boneName=hair_5_r_tgt}}
    lookAt_Axis {axis=<1 0>}
   }
   Save cached pose 'cst_pose'(Pose=Component To Local(ComponentPose=@n65)) {cacheName=cst_pose}
   n66 = "Constraint - hair_neck_02 (P)"(ComponentPose=Local To Component(LocalPose=Copy Pose From Mesh() {bUseAttachedParent=true}), Alpha=1, ConstraintWeights_0=1) {
    alphaBoolBlend {blendOption=Linear}
    boneToModify {boneName=hair_neck_02}
    constraintSetup [{targetBone {boneName=head}}]
    constraintWeights [1]
   }
   Save cached pose 'orig_cst'(Pose=Component To Local(ComponentPose=@n66)) {cacheName=orig_cst}
   Use cached pose 'cst_pose'()
   Use cached pose 'orig_cst'()
  ]}
 EventGraph {nodes [
   Event OnBeginSkydiving()
   > Set skydive_alpha(skydive_alpha:double=1) -> (Output_Get:double)
   Event OnEndSkydiving()
   > Set skydive_alpha(skydive_alpha=0)
   TryGetPawnOwner()
   Event BlueprintUpdateAnimation()
  ]}
]
variables [
 bIsRiding_0 {type=boolean}
 hair_movement {type=number}
 skeletal mesh {type=SkeletalMeshComponent}
 skydive_alpha {type=number}
]
